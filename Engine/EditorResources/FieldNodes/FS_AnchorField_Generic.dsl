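; Chaos anchor field actor that applies a chosen dynamic state to physics objects, such as anchoring destructibles, within a box, sphere or plane.
; The construction script, when active, rebuilds a construction field from a uniform integer culled by the selected shape's falloff.
; AnchorDebugSetup drives the preview: shape visibility, active or deactivated colours, scaled text labels and arrow, hidden in game unless Debug is set.
; BeginPlay and overlap events are empty.
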(event EventBeginPlay)

(event Collision|EventActorBeginOverlap
  (params (OtherActor :type Actor%20Object%20Reference)))

(fn AnchorDebugSetup
  (bind (_returnvalue_x _returnvalue_y _returnvalue_z) (Math|Vector|BreakVector (Transformation|GetWorldScale (Variables|Field|GetFieldSystemComponent))))
  (bind _debug (Variables|Default|GetDebug))
  (bind _returnvalue (not _debug))
  (bind _fieldsystemcomponent (Variables|Field|GetFieldSystemComponent))
  (bind (_returnvalue_x_2 _returnvalue_y_2) (Math|Vector|BreakVector (Transformation|GetWorldScale _fieldsystemcomponent)))
  (bind _returnvalue_2 (Math|Float|Min(Float) _returnvalue_x_2 _returnvalue_y_2))
  (bind _activevaluetext (Variables|Default|GetActiveValueText))
  (bind _activetitletext (Variables|Default|GetActiveTitleText))
  (bind _anchor_title_text (Variables|Default|GetAnchor%20Title%20Text))
  (bind _array (Utilities|Array|MakeArray _anchor_title_text _activetitletext _activevaluetext))
  (bind _anchor_text (Variables|Default|GetAnchor%20Text))
  (Rendering|SetVisibility _anchor_text (and (Variables|Display|GetViewDebugText) _debug) true)
  (Development|SetHiddeninGame _anchor_text _returnvalue true)
  ;[0.3666 0.07801 0.6823 0.25] Function to swap fallloff shape visibility according to selected falloff type
  (CallFunction|SetFalloffVisibility)
  ;[0.3666 0.07801 0.6823 0.25] Falloff Color and Visibility
  (Development|SetHiddeninGame (Utilities|Array|MakeArray (Variables|Default|GetPlaneVolume) (Variables|Default|GetBoxVolumeCol) (Variables|Default|GetSphereVolumeCol)) _returnvalue true)
  (Variables|Default|SetDeactivated (* (Variables|Display|GetAnchorColour) "(R=0.250000,G=0.250000,B=0.250000,A=1.000000)"))
  (switch Utilities|FlowControl|Switch|SwitchonEFieldShapeType (Variables|Default|GetAnchor%20Falloff%20Shape)
    (:NewEnumerator0
      (Variables|Default|SetAnchorVolume (Variables|Default|GetBoxVolume)))
    (:NewEnumerator1
      (Transformation|SetWorldScale3D (Variables|Default|GetSphereVolumeCol) (Math|Conversions|ToVector(Float) _returnvalue_z))
      (Variables|Default|SetAnchorVolume (Variables|Default|GetSphereVolume)))
    (:NewEnumerator2
      (Variables|Default|SetAnchorVolume (Variables|Default|GetPlaneVolume))))
  (bind _returnvalue_3 (Rendering|Material|CreateDynamicMaterialInstance (Variables|Default|GetAnchorVolume) 0 0 "None" :!class /Script/Engine.PrimitiveComponent))
  (if (Variables|Default|GetAnchor%20Active)
    (Rendering|Material|SetVectorParameterValue _returnvalue_3 "Preview_Colour" (Variables|Display|GetAnchorColour) :!class /Script/Engine.MaterialInstanceDynamic)
    ;[0.3666 0.07801 0.6823 0.25] Text Color and Position
    (Rendering|Components|TextRender|SetTextRenderColor _array (Math|Conversions|ToColor(LinearColor) (Variables|Display|GetAnchorColour)))
    (else
      (Rendering|Material|SetVectorParameterValue _returnvalue_3 "Preview_Colour" (Variables|Default|GetDeactivated) :!class /Script/Engine.MaterialInstanceDynamic)
      (Rendering|Components|TextRender|SetTextRenderColor _array (Math|Conversions|ToColor(LinearColor) (Variables|Default|GetDeactivatedText)))))
  (Rendering|Components|TextRender|SetText (Variables|Default|GetActiveValueText) (Utilities|Text|ToText(Boolean) (Variables|Default|GetAnchor%20Active)))
  (Rendering|Components|TextRender|SetText (Variables|Default|GetAnchor%20Title%20Text) (Variables|Display|GetAnchorTextDisplay))
  (Transformation|SetWorldScale3D (Variables|Default|GetAnchor%20Text) (Math|Conversions|ToVector(Float) (* (Variables|Display|GetTextScaleMult) _returnvalue_2)))
  (Transformation|SetWorldScale3D (Variables|Default|GetArrow) (Math|Conversions|ToVector(Float) _returnvalue_2))
  (Transformation|SetRelativeLocation (Variables|Default|GetAnchor%20Text) (Math|Vector|MakeVector 0.0 0.0 (Variables|Display|GetTextVerticalOffset)) false false))

(fn FalloffFieldSwitch
  (params (UniformInt :type Uniform%20Integer%20Object%20Reference))
  ;[1 0.6 0 0.251] Output falloff field results according to user selection
  (switch Utilities|FlowControl|Switch|SwitchonEFieldShapeType (Variables|Default|GetAnchor%20Falloff%20Shape)
    (:NewEnumerator0
      (return (Field|SetCullingField (Variables|Default|GetCullingField_Box) (Field|SetBoxFalloff (Variables|Default|GetBoxFalloff) 1.0 1.0 1.0 0.0 (Transformation|GetWorldTransform (Variables|Default|GetBoxVolumeCol)) "Field_FallOff_None") UniformInt "Field_Culling_Outside")))
    (:NewEnumerator1
      (return (Field|SetCullingField (Variables|Default|GetCullingField_Sphere) (Field|SetRadialFalloff (Variables|Default|GetRadialFalloff) 1.0 1.0 1.0 0.0 (Components|Sphere|GetScaledSphereRadius (Variables|Default|GetSphereVolumeCol)) (Transformation|GetWorldLocation (Variables|Default|GetSphereVolumeCol)) "Field_FallOff_None") UniformInt "Field_Culling_Outside")))
    (:NewEnumerator2
      (return (Field|SetCullingField (Variables|Default|GetCullingField_Plane) (Field|SetPlaneFalloff (Variables|Default|GetPlaneFalloff) 1.0 1.0 1.0 0.0 1e+06 (Transformation|GetWorldLocation (Variables|Default|GetPlaneVolume)) (Transformation|GetUpVector (Variables|Default|GetPlaneVolume)) "Field_FallOff_None") UniformInt "Field_Culling_Outside")))))

(fn SetFalloffVisibility
  (switch Utilities|FlowControl|Switch|SwitchonEFieldShapeType (Variables|Default|GetAnchor%20Falloff%20Shape)
    (:NewEnumerator0
      ;[0.3666 0.07801 0.6823 0.25] Swap Falloff field visibility based on user selection
      (Rendering|SetVisibility (Variables|Default|GetSphereVolumeCol) false true)
      (Rendering|SetVisibility (Variables|Default|GetPlaneVolume) false true)
      (Rendering|SetVisibility (Variables|Default|GetBoxVolumeCol) true true))
    (:NewEnumerator1
      (Rendering|SetVisibility (Variables|Default|GetSphereVolumeCol) true true)
      (Rendering|SetVisibility (Variables|Default|GetPlaneVolume) false true)
      (Rendering|SetVisibility (Variables|Default|GetBoxVolumeCol) false true))
    (:NewEnumerator2
      (Rendering|SetVisibility (Variables|Default|GetSphereVolumeCol) false true)
      (Rendering|SetVisibility (Variables|Default|GetPlaneVolume) true true)
      (Rendering|SetVisibility (Variables|Default|GetBoxVolumeCol) false true)))
  ;[0.3666 0.07801 0.6823 0.25] Optional logic to show/hide wireframe and solid visualization
  (if (not (Variables|Display|GetViewSolidShapes))
    (Rendering|SetVisibility (Variables|Default|GetSphereVolume) false false)
    (Rendering|SetVisibility (Variables|Default|GetBoxVolume) false false)
    (Rendering|SetVisibility (Variables|Default|GetPlaneVolume) false false))
  (if (not (Variables|Display|GetViewWireFrame))
    (Rendering|SetVisibility (Variables|Default|GetSphereVolumeCol) false false)
    (Rendering|SetVisibility (Variables|Default|GetBoxVolumeCol) false false)
    (Rendering|SetVisibility (Variables|Default|GetPlaneVolumeCol) false false)))

(fn ConstructionScript
  ;[0.3666 0.07801 0.6823 0.25 32] Debug Mode Logic
  (CallFunction|AnchorDebugSetup)
  ;[1 0.08854 0.08854 0.25 32] Active?
  (if (Variables|Default|GetAnchor%20Active)
    ;[1 0.6 0 0.243 32] Anchor Logic
    ;;[1 0.6 0 0.243 16] This node refreshes the field system connection when its position or size is altered.
    (Field|RemoveConstructionFields (Variables|Field|GetFieldSystemComponent))
    ;;[1 0.6 0 0.251] Function to use selected falloff volume type
    (bind _cullingfield (CallFunction|FalloffFieldSwitch (Field|SetUniformInteger (Variables|Default|GetUniformInteger) (Math|Conversions|ToInteger(Byte) (Variables|Default|GetDynamic%20State)))))
    ;;[0 1 0.4743 0.251] Apply the Anchor Field here
    (Field|AddConstructionField (Variables|Field|GetFieldSystemComponent) true "Field_DynamicState" 0 _cullingfield)))
