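; Configurable Chaos field system actor that breaks and moves Geometry Collections with transient physics fields.
; BeginPlay, Tick and CE_Trigger run the same optional stages by activation type: dynamic state, external strain, decay (internal strain), radial or directional velocity, and torque.
; Helper functions build noise, falloff and culling fields that shape these stages.
; Lifespan logic deactivates the field, while a pulsing preview material and debug text show its state and settings.
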(event Custom|CE_Trigger
  (bind (_decayfalloffminmax_x _decayfalloffminmax_y) (Math|Vector2D|BreakVector2D (Variables|GetDecayFalloffMinMax_X)))
  ;[0 0.6017 1 0.23] Event Tick/Trigger Variable Setup
  ;;[0 0.6017 1 0.23] Trigger Event Starts Here
  (Variables|Default|SetisTriggered true)
  (if (Variables|Default|GetField%20Active)
    (CallFunction|InitializeFieldVariables)
    (Rendering|Material|SetScalarParameterValue (Variables|Default|GetPreview%20Material) (Variables|Default|GetPulseLevel) 2.0 :!class /Script/Engine.MaterialInstanceDynamic)
    (if (Variables|FieldSettings|AdvancedFieldSettings|ForceDynamicSwitching|GetForce%20Dynamic%20Switch)
      ;[1 1 0 0.15 50] Activate And Break
      ;;[0.2235 1 0 0.25 32] Make Dynamic - This option is not necessary most of the time. Turning on will force activate kinematic and anchored nodes
      (bind _returnvalue (AddComponent|NODE_AddUniformInteger-0))
      (bind _operatorfieldout (Default|FalloffShapeSwitch "Field_FallOff_None" "(X=0.000000,Y=1.000000)"))
      (bind _returnvalue_2 (AddComponent|NODE_AddCullingField-5))
      (if (Variables|FieldSettings|AdvancedFieldSettings|ForceDynamicSwitching|GetActivateTaggedStaticAndSkeletal)
        (Default|MakeDynamicEnableNonGC))
      ;;;[0 1 0.4743 0.251] Field Applied Here
      (Field|AddTransientField (Variables|Field|GetFieldSystemComponent) true "Field_DynamicState" (Field|SetMetaDataResolution (Variables|Default|GetFieldSystemMetaDataProcessingResolution) "Field_Resolution_Maximum") (Field|SetCullingField _returnvalue_2 _operatorfieldout (Field|SetUniformInteger _returnvalue (Math|Conversions|ToInteger(Byte) (Variables|Default|GetDynamic%20State))) "Field_Culling_Outside")))
    (if (Variables|FieldSettings|ApplyExternalStraintoBreakNodeConnections|GetUse%20External%20Strain)
      ;;[1 1 0 0.25] External Strain applied to break connections between geometry collection nodes
      (bind _cullingfield (Default|FalloffandCullSwitchMain (Variables|FieldSettings|ApplyExternalStraintoBreakNodeConnections|GetStrain%20Magnitude) (Variables|FieldSettings|ApplyExternalStraintoBreakNodeConnections|GetStrainFalloffType) (Variables|FieldSettings|ApplyExternalStraintoBreakNodeConnections|GetStrainFalloffMinMax)))
      ;;;[0 1 0.4743 0.251] Field Applied Here
      (Field|AddTransientField (Variables|Field|GetFieldSystemComponent) true "Field_ExternalClusterStrain" 0 _cullingfield))
    (if (Variables|FieldSettings|InternalStrain%28Decay%29|GetUseDecay)
      ;;[1 0.5 0 0.25 32] Internal Strain (Decay): Using internal strain will alter the strength of the internal node connections in a Geometry Collection.
      (bind _output_get (Variables|Default|SetTotalDecay (+ (Variables|FieldSettings|InternalStrain%28Decay%29|GetDecayAmount) (Math|Float|Absolute(Float) (Variables|Default|GetTotalDecay)))))
      (if (< _output_get (Variables|FieldSettings|InternalStrain%28Decay%29|GetMaxDecayAmount))
        (bind _outputpin (CallFunction|CalculateNoise))
        (bind _returnvalue_3 (AddComponent|NODE_AddUniformScalar-0))
        (bind _returnvalue_4 (AddComponent|NODE_AddOperatorField-9))
        (bind _operatorfieldout_2 (Default|FalloffShapeSwitch (Variables|FieldSettings|InternalStrain%28Decay%29|GetDecayFalloffType) (Math|Vector2D|MakeVector2D _decayfalloffminmax_x _decayfalloffminmax_y)))
        (bind _returnvalue_5 (AddComponent|NODE_AddOperatorField-10))
        (bind _operatorfieldout_3 (Default|FalloffShapeSwitch "Field_FallOff_None" "(X=0.000000,Y=1.000000)"))
        (bind _returnvalue_6 (AddComponent|NODE_AddOperatorField-11))
        (bind _returnvalue_7 (AddComponent|NODE_AddCullingField-4))
        (Utilities|FlowControl|Delay (Variables|FieldSettings|InternalStrain%28Decay%29|GetDecayDelay))
        ;;;[0 1 0.4743 0.251] Field Applied Here
        (Field|AddTransientField (Variables|Field|GetFieldSystemComponent) true "Field_InternalClusterStrain" 0 (Field|SetCullingField _returnvalue_7 _operatorfieldout_3 (Field|SetOperatorField _returnvalue_6 1.0 (Field|SetTerminalField (Variables|Default|GetReturnResultsTerminal)) (Field|SetOperatorField _returnvalue_5 1.0 _operatorfieldout_2 (Field|SetOperatorField _returnvalue_4 1.0 _outputpin (Field|SetUniformScalar _returnvalue_3 (Variables|FieldSettings|InternalStrain%28Decay%29|GetDecayAmount)) "Field_Multiply") "Field_Multiply") "Field_Substract") "Field_Culling_Outside"))
        (else
          (Variables|FieldSettings|InternalStrain%28Decay%29|SetUseDecay false))))
    (if (Variables|FieldSettings|ApplyLinearandAngularVelocities|GetUseRadialVector)
      ;[0 0.2833 1 0.15 50] Make it Move: Apply Force and Angular Momentum
      ;;[0 0.3333 1 0.25 32] Apply Radial Vector
      (bind _returnvalue_8 (AddComponent|NODE_AddRadialVector-0))
      (bind _outputpin_2 (CallFunction|CalculateNoise))
      (bind _returnvalue_9 (AddComponent|NODE_AddOperatorField-8))
      (bind _cullingfield_2 (Default|FalloffandCullSwitchMain 1.0 (Variables|FieldSettings|ApplyLinearandAngularVelocities|GetVelocityFieldFalloffType) (Variables|FieldSettings|ApplyLinearandAngularVelocities|GetVelocityFalloffMinMax) (Field|SetOperatorField _returnvalue_9 1.0 (Field|SetRadialVector _returnvalue_8 (Variables|FieldSettings|ApplyLinearandAngularVelocities|GetRadial%20Magnitude) (+ (Transformation|GetWorldLocation (Variables|Field|GetFieldSystemComponent)) (Variables|FieldSettings|AdvancedFieldSettings|VelocityControls|GetRadialPositionOffset))) _outputpin_2 "Field_Multiply")))
      ;;;[0 1 0.4743 0.251] Field Applied Here
      (Field|AddTransientField (Variables|Field|GetFieldSystemComponent) true (Variables|Default|GetLinearPhysicsType) 0 _cullingfield_2))
    (if (Variables|FieldSettings|ApplyLinearandAngularVelocities|GetUseDirectionalVector)
      ;;[0 0 1 0.5 32] Apply Directional Force
      (if (Variables|FieldSettings|AdvancedFieldSettings|VelocityControls|GetOverideDIrectionalVector)
        (Variables|Default|SetDIrectionalVelocityVector (Variables|FieldSettings|AdvancedFieldSettings|VelocityControls|GetDIrectionalVectorOveride))
        (else
          (Variables|Default|SetDIrectionalVelocityVector (Math|Vector|GetUpVector (Transformation|GetWorldRotation (Variables|Field|GetFieldSystemComponent))))))
      (bind _returnvalue_10 (AddComponent|NODE_AddUniformVector-1))
      (bind _outputpin_3 (CallFunction|CalculateNoise))
      (bind _returnvalue_11 (AddComponent|NODE_AddOperatorField-13))
      (bind _cullingfield_3 (Default|FalloffandCullSwitchMain 1.0 (Variables|FieldSettings|ApplyLinearandAngularVelocities|GetVelocityFieldFalloffType) (Variables|FieldSettings|ApplyLinearandAngularVelocities|GetVelocityFalloffMinMax) (Field|SetOperatorField _returnvalue_11 1.0 (Field|SetUniformVector _returnvalue_10 (Variables|FieldSettings|ApplyLinearandAngularVelocities|GetDirectionalMagnitude) (Variables|Default|GetDIrectionalVelocityVector)) _outputpin_3 "Field_Multiply")))
      ;;;[0 1 0.4743 0.251] Field Applied Here
      (Field|AddTransientField (Variables|Field|GetFieldSystemComponent) true (Variables|Default|GetLinearPhysicsType) 0 _cullingfield_3))
    (if (Variables|FieldSettings|ApplyLinearandAngularVelocities|GetUseTorque)
      ;;[0.25 0 1 0.25 32] Apply Angular Momentum
      ;;;[0.25 0 1 0.25] Slight delay added to improve behavior
      (Utilities|FlowControl|Delay 0.01)
      (bind _returnvalue_12 (AddComponent|NODE_AddUniformVector-0))
      (bind _returnvalue_13 (AddComponent|NODE_AddRandomVector-0))
      (bind _returnvalue_14 (AddComponent|NODE_AddOperatorField-14))
      (bind _returnvalue_15 (AddComponent|NODE_AddUniformScalar-1))
      (bind _returnvalue_16 (AddComponent|NODE_AddOperatorField-15))
      (bind _outputpin_4 (CallFunction|CalculateNoise))
      (bind _returnvalue_17 (AddComponent|NODE_AddOperatorField-12))
      (bind _cullingfield_4 (Default|FalloffandCullSwitchMain 1.0 (Variables|FieldSettings|ApplyLinearandAngularVelocities|GetVelocityFieldFalloffType) (Variables|FieldSettings|ApplyLinearandAngularVelocities|GetVelocityFalloffMinMax) (Field|SetOperatorField _returnvalue_17 1.0 (Field|SetOperatorField _returnvalue_16 1.0 (Field|SetUniformScalar _returnvalue_15 (Variables|FieldSettings|ApplyLinearandAngularVelocities|GetTorqueMult)) (Field|SetOperatorField _returnvalue_14 1.0 (Field|SetRandomVector _returnvalue_13 1.0) (Field|SetUniformVector _returnvalue_12 1.0 (Variables|FieldSettings|AdvancedFieldSettings|VelocityControls|GetTorqueVectorOveride)) "Field_Multiply") "Field_Multiply") _outputpin_4 "Field_Multiply")))
      ;;;[0 1 0.4743 0.251] Field Applied Here
      (Field|AddTransientField (Variables|Field|GetFieldSystemComponent) true (Variables|Default|GetAngularPhysicsType) 0 _cullingfield_4))
    ;[0.2 0.2 0.1412 0.2 32] Cleanup and Lifespan Logic
    (switch Utilities|FlowControl|Switch|SwitchonEFieldActivationType (Variables|Default|GetActivationType)
      (:NewEnumerator0
        (if (not (and (Variables|FieldSettings|AdvancedFieldSettings|TimingandLifespan|GetUseLifespan) (Variables|Default|GetField%20Active)))
          (Utilities|FlowControl|Delay 0.2)
          (Rendering|Material|SetScalarParameterValue (Variables|Default|GetPreview%20Material) (Variables|Default|GetPulseLevel) 1.0 :!class /Script/Engine.MaterialInstanceDynamic)))
      (:NewEnumerator2
        _)
      (:NewEnumerator3
        _)
      (:NewEnumerator4
        (if (not (and (Variables|FieldSettings|AdvancedFieldSettings|TimingandLifespan|GetUseLifespan) (Variables|Default|GetField%20Active)))
          (Utilities|FlowControl|Delay 0.2)
          (Rendering|Material|SetVectorParameterValue (Variables|Default|GetPreview%20Material) "Preview_Colour" (Variables|Default|GetDeactivatedColour) :!class /Script/Engine.MaterialInstanceDynamic)
          (Rendering|Material|SetScalarParameterValue (Variables|Default|GetPreview%20Material) (Variables|Default|GetPulseLevel) 1.0 :!class /Script/Engine.MaterialInstanceDynamic)
          (if (Variables|Default|GetDestroyActor)
            (Utilities|FlowControl|Delay 0.2)
            (Actor|DestroyActor)))))
    (if (Variables|FieldSettings|AdvancedFieldSettings|TimingandLifespan|GetUseLifespan)
      (Variables|Default|SetTimeElapsed (+ (Utilities|Time|GetWorldDeltaSeconds) (Variables|Default|GetTimeElapsed)))
      (if (not (< (Variables|Default|GetTimeElapsed) (Variables|FieldSettings|AdvancedFieldSettings|TimingandLifespan|GetFieldLifespan)))
        (Rendering|Material|SetVectorParameterValue (Variables|Default|GetPreview%20Material) "Preview_Colour" (Variables|Default|GetDeactivatedColour) :!class /Script/Engine.MaterialInstanceDynamic)
        (Rendering|Material|SetScalarParameterValue (Variables|Default|GetPreview%20Material) (Variables|Default|GetPulseLevel) 1.0 :!class /Script/Engine.MaterialInstanceDynamic)
        (Variables|Default|SetField%20Active false)))))

(event EventTick
  (params (DeltaSeconds :type Float%20%28single-precision%29))
  (bind (_decayfalloffminmax_x _decayfalloffminmax_y) (Math|Vector2D|BreakVector2D (Variables|GetDecayFalloffMinMax_X)))
  ;[1 0.08854 0.08854 0.25] Is Active?
  (if (Variables|Default|GetField%20Active)
    ;[0 0.6017 1 0.23] Event Tick/Trigger Variable Setup
    (switch Utilities|FlowControl|Switch|SwitchonEFieldActivationType (Variables|Default|GetActivationType)
      (:NewEnumerator0
        (if (Variables|FieldSettings|AdvancedFieldSettings|TimingandLifespan|GetUseLifespan)
          (Utilities|FlowControl|Gate true :!label _label
            (:Exit
              _))))
      (:NewEnumerator2
        _)
      (:NewEnumerator3
        (goto _label))
      (:NewEnumerator4
        (if (and (Variables|FieldSettings|AdvancedFieldSettings|TimingandLifespan|GetUseLifespan) (Variables|Default|GetisTriggered))
          _)))
    (Rendering|Material|SetScalarParameterValue (Variables|Default|GetPreview%20Material) (Variables|Default|GetPulseLevel) 2.0 :!class /Script/Engine.MaterialInstanceDynamic)
    (if (Variables|FieldSettings|AdvancedFieldSettings|ForceDynamicSwitching|GetForce%20Dynamic%20Switch)
      ;[1 1 0 0.15 50] Activate And Break
      ;;[0.2235 1 0 0.25 32] Make Dynamic - This option is not necessary most of the time. Turning on will force activate kinematic and anchored nodes
      (bind _returnvalue (AddComponent|NODE_AddUniformInteger-0))
      (bind _operatorfieldout (Default|FalloffShapeSwitch "Field_FallOff_None" "(X=0.000000,Y=1.000000)"))
      (bind _returnvalue_2 (AddComponent|NODE_AddCullingField-5))
      (if (Variables|FieldSettings|AdvancedFieldSettings|ForceDynamicSwitching|GetActivateTaggedStaticAndSkeletal)
        (Default|MakeDynamicEnableNonGC))
      ;;;[0 1 0.4743 0.251] Field Applied Here
      (Field|AddTransientField (Variables|Field|GetFieldSystemComponent) true "Field_DynamicState" (Field|SetMetaDataResolution (Variables|Default|GetFieldSystemMetaDataProcessingResolution) "Field_Resolution_Maximum") (Field|SetCullingField _returnvalue_2 _operatorfieldout (Field|SetUniformInteger _returnvalue (Math|Conversions|ToInteger(Byte) (Variables|Default|GetDynamic%20State))) "Field_Culling_Outside")))
    (if (Variables|FieldSettings|ApplyExternalStraintoBreakNodeConnections|GetUse%20External%20Strain)
      ;;[1 1 0 0.25] External Strain applied to break connections between geometry collection nodes
      (bind _cullingfield (Default|FalloffandCullSwitchMain (Variables|FieldSettings|ApplyExternalStraintoBreakNodeConnections|GetStrain%20Magnitude) (Variables|FieldSettings|ApplyExternalStraintoBreakNodeConnections|GetStrainFalloffType) (Variables|FieldSettings|ApplyExternalStraintoBreakNodeConnections|GetStrainFalloffMinMax)))
      ;;;[0 1 0.4743 0.251] Field Applied Here
      (Field|AddTransientField (Variables|Field|GetFieldSystemComponent) true "Field_ExternalClusterStrain" 0 _cullingfield))
    (if (Variables|FieldSettings|InternalStrain%28Decay%29|GetUseDecay)
      ;;[1 0.5 0 0.25 32] Internal Strain (Decay): Using internal strain will alter the strength of the internal node connections in a Geometry Collection.
      (bind _output_get (Variables|Default|SetTotalDecay (+ (Variables|FieldSettings|InternalStrain%28Decay%29|GetDecayAmount) (Math|Float|Absolute(Float) (Variables|Default|GetTotalDecay)))))
      (if (< _output_get (Variables|FieldSettings|InternalStrain%28Decay%29|GetMaxDecayAmount))
        (bind _outputpin (CallFunction|CalculateNoise))
        (bind _returnvalue_3 (AddComponent|NODE_AddUniformScalar-0))
        (bind _returnvalue_4 (AddComponent|NODE_AddOperatorField-9))
        (bind _operatorfieldout_2 (Default|FalloffShapeSwitch (Variables|FieldSettings|InternalStrain%28Decay%29|GetDecayFalloffType) (Math|Vector2D|MakeVector2D _decayfalloffminmax_x _decayfalloffminmax_y)))
        (bind _returnvalue_5 (AddComponent|NODE_AddOperatorField-10))
        (bind _operatorfieldout_3 (Default|FalloffShapeSwitch "Field_FallOff_None" "(X=0.000000,Y=1.000000)"))
        (bind _returnvalue_6 (AddComponent|NODE_AddOperatorField-11))
        (bind _returnvalue_7 (AddComponent|NODE_AddCullingField-4))
        (Utilities|FlowControl|Delay (Variables|FieldSettings|InternalStrain%28Decay%29|GetDecayDelay))
        ;;;[0 1 0.4743 0.251] Field Applied Here
        (Field|AddTransientField (Variables|Field|GetFieldSystemComponent) true "Field_InternalClusterStrain" 0 (Field|SetCullingField _returnvalue_7 _operatorfieldout_3 (Field|SetOperatorField _returnvalue_6 1.0 (Field|SetTerminalField (Variables|Default|GetReturnResultsTerminal)) (Field|SetOperatorField _returnvalue_5 1.0 _operatorfieldout_2 (Field|SetOperatorField _returnvalue_4 1.0 _outputpin (Field|SetUniformScalar _returnvalue_3 (Variables|FieldSettings|InternalStrain%28Decay%29|GetDecayAmount)) "Field_Multiply") "Field_Multiply") "Field_Substract") "Field_Culling_Outside"))
        (else
          (Variables|FieldSettings|InternalStrain%28Decay%29|SetUseDecay false))))
    (if (Variables|FieldSettings|ApplyLinearandAngularVelocities|GetUseRadialVector)
      ;[0 0.2833 1 0.15 50] Make it Move: Apply Force and Angular Momentum
      ;;[0 0.3333 1 0.25 32] Apply Radial Vector
      (bind _returnvalue_8 (AddComponent|NODE_AddRadialVector-0))
      (bind _outputpin_2 (CallFunction|CalculateNoise))
      (bind _returnvalue_9 (AddComponent|NODE_AddOperatorField-8))
      (bind _cullingfield_2 (Default|FalloffandCullSwitchMain 1.0 (Variables|FieldSettings|ApplyLinearandAngularVelocities|GetVelocityFieldFalloffType) (Variables|FieldSettings|ApplyLinearandAngularVelocities|GetVelocityFalloffMinMax) (Field|SetOperatorField _returnvalue_9 1.0 (Field|SetRadialVector _returnvalue_8 (Variables|FieldSettings|ApplyLinearandAngularVelocities|GetRadial%20Magnitude) (+ (Transformation|GetWorldLocation (Variables|Field|GetFieldSystemComponent)) (Variables|FieldSettings|AdvancedFieldSettings|VelocityControls|GetRadialPositionOffset))) _outputpin_2 "Field_Multiply")))
      ;;;[0 1 0.4743 0.251] Field Applied Here
      (Field|AddTransientField (Variables|Field|GetFieldSystemComponent) true (Variables|Default|GetLinearPhysicsType) 0 _cullingfield_2))
    (if (Variables|FieldSettings|ApplyLinearandAngularVelocities|GetUseDirectionalVector)
      ;;[0 0 1 0.5 32] Apply Directional Force
      (if (Variables|FieldSettings|AdvancedFieldSettings|VelocityControls|GetOverideDIrectionalVector)
        (Variables|Default|SetDIrectionalVelocityVector (Variables|FieldSettings|AdvancedFieldSettings|VelocityControls|GetDIrectionalVectorOveride))
        (else
          (Variables|Default|SetDIrectionalVelocityVector (Math|Vector|GetUpVector (Transformation|GetWorldRotation (Variables|Field|GetFieldSystemComponent))))))
      (bind _returnvalue_10 (AddComponent|NODE_AddUniformVector-1))
      (bind _outputpin_3 (CallFunction|CalculateNoise))
      (bind _returnvalue_11 (AddComponent|NODE_AddOperatorField-13))
      (bind _cullingfield_3 (Default|FalloffandCullSwitchMain 1.0 (Variables|FieldSettings|ApplyLinearandAngularVelocities|GetVelocityFieldFalloffType) (Variables|FieldSettings|ApplyLinearandAngularVelocities|GetVelocityFalloffMinMax) (Field|SetOperatorField _returnvalue_11 1.0 (Field|SetUniformVector _returnvalue_10 (Variables|FieldSettings|ApplyLinearandAngularVelocities|GetDirectionalMagnitude) (Variables|Default|GetDIrectionalVelocityVector)) _outputpin_3 "Field_Multiply")))
      ;;;[0 1 0.4743 0.251] Field Applied Here
      (Field|AddTransientField (Variables|Field|GetFieldSystemComponent) true (Variables|Default|GetLinearPhysicsType) 0 _cullingfield_3))
    (if (Variables|FieldSettings|ApplyLinearandAngularVelocities|GetUseTorque)
      ;;[0.25 0 1 0.25 32] Apply Angular Momentum
      ;;;[0.25 0 1 0.25] Slight delay added to improve behavior
      (Utilities|FlowControl|Delay 0.01)
      (bind _returnvalue_12 (AddComponent|NODE_AddUniformVector-0))
      (bind _returnvalue_13 (AddComponent|NODE_AddRandomVector-0))
      (bind _returnvalue_14 (AddComponent|NODE_AddOperatorField-14))
      (bind _returnvalue_15 (AddComponent|NODE_AddUniformScalar-1))
      (bind _returnvalue_16 (AddComponent|NODE_AddOperatorField-15))
      (bind _outputpin_4 (CallFunction|CalculateNoise))
      (bind _returnvalue_17 (AddComponent|NODE_AddOperatorField-12))
      (bind _cullingfield_4 (Default|FalloffandCullSwitchMain 1.0 (Variables|FieldSettings|ApplyLinearandAngularVelocities|GetVelocityFieldFalloffType) (Variables|FieldSettings|ApplyLinearandAngularVelocities|GetVelocityFalloffMinMax) (Field|SetOperatorField _returnvalue_17 1.0 (Field|SetOperatorField _returnvalue_16 1.0 (Field|SetUniformScalar _returnvalue_15 (Variables|FieldSettings|ApplyLinearandAngularVelocities|GetTorqueMult)) (Field|SetOperatorField _returnvalue_14 1.0 (Field|SetRandomVector _returnvalue_13 1.0) (Field|SetUniformVector _returnvalue_12 1.0 (Variables|FieldSettings|AdvancedFieldSettings|VelocityControls|GetTorqueVectorOveride)) "Field_Multiply") "Field_Multiply") _outputpin_4 "Field_Multiply")))
      ;;;[0 1 0.4743 0.251] Field Applied Here
      (Field|AddTransientField (Variables|Field|GetFieldSystemComponent) true (Variables|Default|GetAngularPhysicsType) 0 _cullingfield_4))
    ;[0.2 0.2 0.1412 0.2 32] Cleanup and Lifespan Logic
    (switch Utilities|FlowControl|Switch|SwitchonEFieldActivationType (Variables|Default|GetActivationType)
      (:NewEnumerator0
        (if (not (and (Variables|FieldSettings|AdvancedFieldSettings|TimingandLifespan|GetUseLifespan) (Variables|Default|GetField%20Active)))
          (Utilities|FlowControl|Delay 0.2)
          (Rendering|Material|SetScalarParameterValue (Variables|Default|GetPreview%20Material) (Variables|Default|GetPulseLevel) 1.0 :!class /Script/Engine.MaterialInstanceDynamic)))
      (:NewEnumerator2
        _)
      (:NewEnumerator3
        _)
      (:NewEnumerator4
        (if (not (and (Variables|FieldSettings|AdvancedFieldSettings|TimingandLifespan|GetUseLifespan) (Variables|Default|GetField%20Active)))
          (Utilities|FlowControl|Delay 0.2)
          (Rendering|Material|SetVectorParameterValue (Variables|Default|GetPreview%20Material) "Preview_Colour" (Variables|Default|GetDeactivatedColour) :!class /Script/Engine.MaterialInstanceDynamic)
          (Rendering|Material|SetScalarParameterValue (Variables|Default|GetPreview%20Material) (Variables|Default|GetPulseLevel) 1.0 :!class /Script/Engine.MaterialInstanceDynamic)
          (if (Variables|Default|GetDestroyActor)
            (Utilities|FlowControl|Delay 0.2)
            (Actor|DestroyActor)))))
    (if (Variables|FieldSettings|AdvancedFieldSettings|TimingandLifespan|GetUseLifespan)
      (Variables|Default|SetTimeElapsed (+ (Utilities|Time|GetWorldDeltaSeconds) (Variables|Default|GetTimeElapsed)))
      (if (not (< (Variables|Default|GetTimeElapsed) (Variables|FieldSettings|AdvancedFieldSettings|TimingandLifespan|GetFieldLifespan)))
        (Rendering|Material|SetVectorParameterValue (Variables|Default|GetPreview%20Material) "Preview_Colour" (Variables|Default|GetDeactivatedColour) :!class /Script/Engine.MaterialInstanceDynamic)
        (Rendering|Material|SetScalarParameterValue (Variables|Default|GetPreview%20Material) (Variables|Default|GetPulseLevel) 1.0 :!class /Script/Engine.MaterialInstanceDynamic)
        (Variables|Default|SetField%20Active false)))))

(event EventBeginPlay
  (bind (_decayfalloffminmax_x _decayfalloffminmax_y) (Math|Vector2D|BreakVector2D (Variables|GetDecayFalloffMinMax_X)))
  ;[1 0.08854 0.08854 0.25] Is Active?
  (if (Variables|Default|GetField%20Active)
    ;[0 0.6017 1 0.23] Begin Play Setup
    ;;[0 0.6017 1 0.23] Set up variables
    (CallFunction|InitializeFieldVariables)
    (switch Utilities|FlowControl|Switch|SwitchonEFieldActivationType (Variables|Default|GetActivationType)
      (:NewEnumerator0
        (if (Variables|FieldSettings|AdvancedFieldSettings|TimingandLifespan|GetUseLifespan)
          ;[0 0.6017 1 0.23] Event Tick/Trigger Variable Setup
          (Utilities|FlowControl|Delay (Variables|Default|GetDelayAmount) :!label _label)
          (Utilities|FlowControl|Gate true
            (:Exit
              (Rendering|Material|SetScalarParameterValue (Variables|Default|GetPreview%20Material) (Variables|Default|GetPulseLevel) 2.0 :!class /Script/Engine.MaterialInstanceDynamic)))
          (else
            (Utilities|FlowControl|Delay (Variables|Default|GetDelayAmount))
            (Rendering|Material|SetScalarParameterValue (Variables|Default|GetPreview%20Material) (Variables|Default|GetPulseLevel) 55.0 :!class /Script/Engine.MaterialInstanceDynamic)))
        (if (Variables|FieldSettings|AdvancedFieldSettings|ForceDynamicSwitching|GetForce%20Dynamic%20Switch)
          ;[1 1 0 0.15 50] Activate And Break
          ;;[0.2235 1 0 0.25 32] Make Dynamic - This option is not necessary most of the time. Turning on will force activate kinematic and anchored nodes
          (bind _returnvalue (AddComponent|NODE_AddUniformInteger-0))
          (bind _operatorfieldout (Default|FalloffShapeSwitch "Field_FallOff_None" "(X=0.000000,Y=1.000000)"))
          (bind _returnvalue_2 (AddComponent|NODE_AddCullingField-5))
          (if (Variables|FieldSettings|AdvancedFieldSettings|ForceDynamicSwitching|GetActivateTaggedStaticAndSkeletal)
            (Default|MakeDynamicEnableNonGC))
          ;;;[0 1 0.4743 0.251] Field Applied Here
          (Field|AddTransientField (Variables|Field|GetFieldSystemComponent) true "Field_DynamicState" (Field|SetMetaDataResolution (Variables|Default|GetFieldSystemMetaDataProcessingResolution) "Field_Resolution_Maximum") (Field|SetCullingField _returnvalue_2 _operatorfieldout (Field|SetUniformInteger _returnvalue (Math|Conversions|ToInteger(Byte) (Variables|Default|GetDynamic%20State))) "Field_Culling_Outside")))
        (if (Variables|FieldSettings|ApplyExternalStraintoBreakNodeConnections|GetUse%20External%20Strain)
          ;;[1 1 0 0.25] External Strain applied to break connections between geometry collection nodes
          (bind _cullingfield (Default|FalloffandCullSwitchMain (Variables|FieldSettings|ApplyExternalStraintoBreakNodeConnections|GetStrain%20Magnitude) (Variables|FieldSettings|ApplyExternalStraintoBreakNodeConnections|GetStrainFalloffType) (Variables|FieldSettings|ApplyExternalStraintoBreakNodeConnections|GetStrainFalloffMinMax)))
          ;;;[0 1 0.4743 0.251] Field Applied Here
          (Field|AddTransientField (Variables|Field|GetFieldSystemComponent) true "Field_ExternalClusterStrain" 0 _cullingfield))
        (if (Variables|FieldSettings|InternalStrain%28Decay%29|GetUseDecay)
          ;;[1 0.5 0 0.25 32] Internal Strain (Decay): Using internal strain will alter the strength of the internal node connections in a Geometry Collection.
          (bind _output_get (Variables|Default|SetTotalDecay (+ (Variables|FieldSettings|InternalStrain%28Decay%29|GetDecayAmount) (Math|Float|Absolute(Float) (Variables|Default|GetTotalDecay)))))
          (if (< _output_get (Variables|FieldSettings|InternalStrain%28Decay%29|GetMaxDecayAmount))
            (bind _outputpin (CallFunction|CalculateNoise))
            (bind _returnvalue_3 (AddComponent|NODE_AddUniformScalar-0))
            (bind _returnvalue_4 (AddComponent|NODE_AddOperatorField-9))
            (bind _operatorfieldout_2 (Default|FalloffShapeSwitch (Variables|FieldSettings|InternalStrain%28Decay%29|GetDecayFalloffType) (Math|Vector2D|MakeVector2D _decayfalloffminmax_x _decayfalloffminmax_y)))
            (bind _returnvalue_5 (AddComponent|NODE_AddOperatorField-10))
            (bind _operatorfieldout_3 (Default|FalloffShapeSwitch "Field_FallOff_None" "(X=0.000000,Y=1.000000)"))
            (bind _returnvalue_6 (AddComponent|NODE_AddOperatorField-11))
            (bind _returnvalue_7 (AddComponent|NODE_AddCullingField-4))
            (Utilities|FlowControl|Delay (Variables|FieldSettings|InternalStrain%28Decay%29|GetDecayDelay))
            ;;;[0 1 0.4743 0.251] Field Applied Here
            (Field|AddTransientField (Variables|Field|GetFieldSystemComponent) true "Field_InternalClusterStrain" 0 (Field|SetCullingField _returnvalue_7 _operatorfieldout_3 (Field|SetOperatorField _returnvalue_6 1.0 (Field|SetTerminalField (Variables|Default|GetReturnResultsTerminal)) (Field|SetOperatorField _returnvalue_5 1.0 _operatorfieldout_2 (Field|SetOperatorField _returnvalue_4 1.0 _outputpin (Field|SetUniformScalar _returnvalue_3 (Variables|FieldSettings|InternalStrain%28Decay%29|GetDecayAmount)) "Field_Multiply") "Field_Multiply") "Field_Substract") "Field_Culling_Outside"))
            (else
              (Variables|FieldSettings|InternalStrain%28Decay%29|SetUseDecay false))))
        (if (Variables|FieldSettings|ApplyLinearandAngularVelocities|GetUseRadialVector)
          ;[0 0.2833 1 0.15 50] Make it Move: Apply Force and Angular Momentum
          ;;[0 0.3333 1 0.25 32] Apply Radial Vector
          (bind _returnvalue_8 (AddComponent|NODE_AddRadialVector-0))
          (bind _outputpin_2 (CallFunction|CalculateNoise))
          (bind _returnvalue_9 (AddComponent|NODE_AddOperatorField-8))
          (bind _cullingfield_2 (Default|FalloffandCullSwitchMain 1.0 (Variables|FieldSettings|ApplyLinearandAngularVelocities|GetVelocityFieldFalloffType) (Variables|FieldSettings|ApplyLinearandAngularVelocities|GetVelocityFalloffMinMax) (Field|SetOperatorField _returnvalue_9 1.0 (Field|SetRadialVector _returnvalue_8 (Variables|FieldSettings|ApplyLinearandAngularVelocities|GetRadial%20Magnitude) (+ (Transformation|GetWorldLocation (Variables|Field|GetFieldSystemComponent)) (Variables|FieldSettings|AdvancedFieldSettings|VelocityControls|GetRadialPositionOffset))) _outputpin_2 "Field_Multiply")))
          ;;;[0 1 0.4743 0.251] Field Applied Here
          (Field|AddTransientField (Variables|Field|GetFieldSystemComponent) true (Variables|Default|GetLinearPhysicsType) 0 _cullingfield_2))
        (if (Variables|FieldSettings|ApplyLinearandAngularVelocities|GetUseDirectionalVector)
          ;;[0 0 1 0.5 32] Apply Directional Force
          (if (Variables|FieldSettings|AdvancedFieldSettings|VelocityControls|GetOverideDIrectionalVector)
            (Variables|Default|SetDIrectionalVelocityVector (Variables|FieldSettings|AdvancedFieldSettings|VelocityControls|GetDIrectionalVectorOveride))
            (else
              (Variables|Default|SetDIrectionalVelocityVector (Math|Vector|GetUpVector (Transformation|GetWorldRotation (Variables|Field|GetFieldSystemComponent))))))
          (bind _returnvalue_10 (AddComponent|NODE_AddUniformVector-1))
          (bind _outputpin_3 (CallFunction|CalculateNoise))
          (bind _returnvalue_11 (AddComponent|NODE_AddOperatorField-13))
          (bind _cullingfield_3 (Default|FalloffandCullSwitchMain 1.0 (Variables|FieldSettings|ApplyLinearandAngularVelocities|GetVelocityFieldFalloffType) (Variables|FieldSettings|ApplyLinearandAngularVelocities|GetVelocityFalloffMinMax) (Field|SetOperatorField _returnvalue_11 1.0 (Field|SetUniformVector _returnvalue_10 (Variables|FieldSettings|ApplyLinearandAngularVelocities|GetDirectionalMagnitude) (Variables|Default|GetDIrectionalVelocityVector)) _outputpin_3 "Field_Multiply")))
          ;;;[0 1 0.4743 0.251] Field Applied Here
          (Field|AddTransientField (Variables|Field|GetFieldSystemComponent) true (Variables|Default|GetLinearPhysicsType) 0 _cullingfield_3))
        (if (Variables|FieldSettings|ApplyLinearandAngularVelocities|GetUseTorque)
          ;;[0.25 0 1 0.25 32] Apply Angular Momentum
          ;;;[0.25 0 1 0.25] Slight delay added to improve behavior
          (Utilities|FlowControl|Delay 0.01)
          (bind _returnvalue_12 (AddComponent|NODE_AddUniformVector-0))
          (bind _returnvalue_13 (AddComponent|NODE_AddRandomVector-0))
          (bind _returnvalue_14 (AddComponent|NODE_AddOperatorField-14))
          (bind _returnvalue_15 (AddComponent|NODE_AddUniformScalar-1))
          (bind _returnvalue_16 (AddComponent|NODE_AddOperatorField-15))
          (bind _outputpin_4 (CallFunction|CalculateNoise))
          (bind _returnvalue_17 (AddComponent|NODE_AddOperatorField-12))
          (bind _cullingfield_4 (Default|FalloffandCullSwitchMain 1.0 (Variables|FieldSettings|ApplyLinearandAngularVelocities|GetVelocityFieldFalloffType) (Variables|FieldSettings|ApplyLinearandAngularVelocities|GetVelocityFalloffMinMax) (Field|SetOperatorField _returnvalue_17 1.0 (Field|SetOperatorField _returnvalue_16 1.0 (Field|SetUniformScalar _returnvalue_15 (Variables|FieldSettings|ApplyLinearandAngularVelocities|GetTorqueMult)) (Field|SetOperatorField _returnvalue_14 1.0 (Field|SetRandomVector _returnvalue_13 1.0) (Field|SetUniformVector _returnvalue_12 1.0 (Variables|FieldSettings|AdvancedFieldSettings|VelocityControls|GetTorqueVectorOveride)) "Field_Multiply") "Field_Multiply") _outputpin_4 "Field_Multiply")))
          ;;;[0 1 0.4743 0.251] Field Applied Here
          (Field|AddTransientField (Variables|Field|GetFieldSystemComponent) true (Variables|Default|GetAngularPhysicsType) 0 _cullingfield_4))
        ;[0.2 0.2 0.1412 0.2 32] Cleanup and Lifespan Logic
        (switch Utilities|FlowControl|Switch|SwitchonEFieldActivationType (Variables|Default|GetActivationType)
          (:NewEnumerator0
            (if (not (and (Variables|FieldSettings|AdvancedFieldSettings|TimingandLifespan|GetUseLifespan) (Variables|Default|GetField%20Active)))
              (Utilities|FlowControl|Delay 0.2)
              (Rendering|Material|SetScalarParameterValue (Variables|Default|GetPreview%20Material) (Variables|Default|GetPulseLevel) 1.0 :!class /Script/Engine.MaterialInstanceDynamic)))
          (:NewEnumerator2
            _)
          (:NewEnumerator3
            _)
          (:NewEnumerator4
            (if (not (and (Variables|FieldSettings|AdvancedFieldSettings|TimingandLifespan|GetUseLifespan) (Variables|Default|GetField%20Active)))
              (Utilities|FlowControl|Delay 0.2)
              (Rendering|Material|SetVectorParameterValue (Variables|Default|GetPreview%20Material) "Preview_Colour" (Variables|Default|GetDeactivatedColour) :!class /Script/Engine.MaterialInstanceDynamic)
              (Rendering|Material|SetScalarParameterValue (Variables|Default|GetPreview%20Material) (Variables|Default|GetPulseLevel) 1.0 :!class /Script/Engine.MaterialInstanceDynamic)
              (if (Variables|Default|GetDestroyActor)
                (Utilities|FlowControl|Delay 0.2)
                (Actor|DestroyActor)))))
        (if (Variables|FieldSettings|AdvancedFieldSettings|TimingandLifespan|GetUseLifespan)
          (Variables|Default|SetTimeElapsed (+ (Utilities|Time|GetWorldDeltaSeconds) (Variables|Default|GetTimeElapsed)))
          (if (not (< (Variables|Default|GetTimeElapsed) (Variables|FieldSettings|AdvancedFieldSettings|TimingandLifespan|GetFieldLifespan)))
            (Rendering|Material|SetVectorParameterValue (Variables|Default|GetPreview%20Material) "Preview_Colour" (Variables|Default|GetDeactivatedColour) :!class /Script/Engine.MaterialInstanceDynamic)
            (Rendering|Material|SetScalarParameterValue (Variables|Default|GetPreview%20Material) (Variables|Default|GetPulseLevel) 1.0 :!class /Script/Engine.MaterialInstanceDynamic)
            (Variables|Default|SetField%20Active false))))
      (:NewEnumerator3
        _))
    (goto _label)))

(fn CalculateNoise
  (bind (_returnvalue_location _returnvalue_rotation) (Math|Transform|BreakTransform (Transformation|GetWorldTransform (Variables|Field|GetFieldSystemComponent))))
  (bind (_noiseminmax_x _noiseminmax_y) (Math|Vector2D|BreakVector2D (Variables|GetNoiseMinMax_X)))
  (bind _returnvalue (AddComponent|NODE_AddNoiseField-0))
  (return (Field|SetNoiseField _returnvalue _noiseminmax_x _noiseminmax_y (Math|Transform|MakeTransform _returnvalue_location _returnvalue_rotation (* (Variables|Default|GetNoiseScaleBase) (Variables|FieldSettings|Noise|GetNoiseScaleMult))))))

(fn DisplayTextSetup
  (bind _textdisplay (Variables|Default|GetTextDisplay))
  (bind _output (Utilities|Array|Get(aref) _textdisplay 3))
  (bind _noiseminmax_x (Math|Vector2D|BreakVector2D (Variables|GetNoiseMinMax_X)))
  (bind (_noiseminmax_x_2 _noiseminmax_y) (Math|Vector2D|BreakVector2D (Variables|GetNoiseMinMax_X)))
  (Variables|Default|SetAllText (Utilities|Array|MakeArray (Variables|Default|GetFieldTitleText) (Variables|Default|GetActiveTitleText) (Variables|Default|GetActiveValueText) (Variables|Default|GetTorqueTitleText) (Variables|Default|GetTorqueValueText) (Variables|Default|GetDirectionalMagTitleText) (Variables|Default|GetDirectionalValueText) (Variables|Default|GetNoise%20MinMaxTitle%20Text) (Variables|Default|GetNoiseMaxValueText) (Variables|Default|GetNoiseMinValueText) (Variables|Default|GetDelayValueText) (Variables|Default|GetDelayTitleText) (Variables|Default|GetRadial%20Mag%20Title%20Text) (Variables|Default|GetRadialValueText) (Variables|Default|GetStrainTitleText) (Variables|Default|GetStrainValueText)))
  ;[0.3666 0.07801 0.6823 0.25 32] Set Debug Text Values
  (Rendering|Components|TextRender|SetText (Variables|Default|GetFieldTitleText) (Variables|FieldSettings|Display|GetChaos%20Field%20Name))
  (Rendering|Components|TextRender|SetText (Variables|Default|GetActiveValueText) (Utilities|Text|ToText(Boolean) (Variables|Default|GetField%20Active)))
  (switch Utilities|FlowControl|Switch|SwitchonEFieldActivationType (Variables|Default|GetActivationType)
    (:NewEnumerator0
      (Rendering|Components|TextRender|SetText (Variables|Default|GetDelayValueText) (Utilities|Text|ToText(Float) (Variables|Default|GetDelayAmount)) :!label _label))
    (:NewEnumerator2
      (Rendering|Components|TextRender|SetText (Variables|Default|GetDelayValueText) (Utilities|Array|Get(aref) (Variables|Default|GetTextDisplay))))
    (:NewEnumerator3
      (goto _label))
    (:NewEnumerator4
      (Rendering|Components|TextRender|SetText (Variables|Default|GetDelayValueText) (Utilities|Array|Get(aref) (Variables|Default|GetTextDisplay) 4))))
  (Rendering|Components|TextRender|SetText (Variables|Default|GetStrainValueText) (Utilities|Text|ToText(Float) (Variables|FieldSettings|ApplyExternalStraintoBreakNodeConnections|GetStrain%20Magnitude)))
  (if (Variables|FieldSettings|ApplyLinearandAngularVelocities|GetUseRadialVector)
    (Rendering|Components|TextRender|SetText (Variables|Default|GetRadialValueText) (Utilities|Text|ToText(Float) (Variables|FieldSettings|ApplyLinearandAngularVelocities|GetRadial%20Magnitude)))
    (else
      (Rendering|Components|TextRender|SetText (Variables|Default|GetRadialValueText) (Utilities|Array|Get(aref) (Variables|Default|GetTextDisplay) 1))))
  (if (Variables|FieldSettings|ApplyLinearandAngularVelocities|GetUseDirectionalVector)
    (Rendering|Components|TextRender|SetText (Variables|Default|GetDirectionalValueText) (Utilities|Text|ToText(Float) (Variables|FieldSettings|ApplyLinearandAngularVelocities|GetDirectionalMagnitude)))
    (else
      (Rendering|Components|TextRender|SetText (Variables|Default|GetDirectionalValueText) (Utilities|Array|Get(aref) (Variables|Default|GetTextDisplay) 1))))
  (if (Variables|FieldSettings|Noise|GetUseNoise)
    (Rendering|Components|TextRender|SetText (Variables|Default|GetNoiseMaxValueText) (Utilities|Text|ToText(Float) _noiseminmax_y))
    (Rendering|Components|TextRender|SetText (Variables|Default|GetNoiseMinValueText) (Utilities|Text|ToText(Float) _noiseminmax_x))
    (else
      (Rendering|Components|TextRender|SetText (Variables|Default|GetNoise%20MinMaxTitle%20Text) (Utilities|Array|Get(aref) (Variables|Default|GetTextDisplay) 2))
      (Rendering|Components|TextRender|SetText (Variables|Default|GetNoiseMaxValueText) _output)
      (Rendering|Components|TextRender|SetText (Variables|Default|GetNoiseMinValueText) _output)))
  (if (Variables|FieldSettings|ApplyLinearandAngularVelocities|GetUseTorque)
    (Rendering|Components|TextRender|SetText (Variables|Default|GetTorqueValueText) (Utilities|Text|ToText(Float) (Variables|FieldSettings|ApplyLinearandAngularVelocities|GetTorqueMult)))
    (else
      (Rendering|Components|TextRender|SetText (Variables|Default|GetTorqueValueText) (Utilities|Array|Get(aref) (Variables|Default|GetTextDisplay) 1)))))

(fn FalloffAndCullSwitch_Main
  (params (Magnitude :type Float%20%28double-precision%29) (FalloffType :type EFieldFalloffType%20Enum :default "Field_FallOff_None") (FalloffMinMax :type Vector%202D%20Structure) (OperatorField :type Operator%20Field%20Object%20Reference))
  (bind _planevolume (Variables|Default|GetPlaneVolume))
  (bind _returnvalue (Transformation|GetUpVector _planevolume))
  (bind _returnvalue_2 (Transformation|GetWorldLocation _planevolume))
  (bind (_falloffminmax_input_x _falloffminmax_input_y) (Math|Vector2D|BreakVector2D (Variables|GetFalloffMinMax_Input_X)))
  (bind _planarfalloffdist (Variables|Default|GetPlanarFalloffDist))
  (bind (_falloffminmax_input_x_2 _falloffminmax_input_y_2) (Math|Vector2D|BreakVector2D (Variables|GetFalloffMinMax_Input_X)))
  (bind _magnitude_input (Variables|Default|GetMagnitude_Input))
  (bind (_falloffminmax_input_x_3 _falloffminmax_input_y_3) (Math|Vector2D|BreakVector2D (Variables|GetFalloffMinMax_Input_X)))
  (bind _operatorfield_input (Variables|Default|GetOperatorFIeld_Input))
  (bind _fieldfallofftype_input (Variables|Default|GetFieldFalloffType_Input))
  (Variables|Default|SetMagnitude_Input Magnitude)
  (Variables|Default|SetFieldFalloffType_Input FalloffType)
  (Variables|Default|SetFalloffMinMax_Input FalloffMinMax)
  (Variables|Default|SetOperatorFIeld_Input OperatorField)
  (switch Utilities|FlowControl|Switch|SwitchonEFieldShapeType (Variables|Default|GetField%20Falloff%20Shape)
    (:NewEnumerator0
      (bind _returnvalue_3 (AddComponent|NODE_AddBoxFalloff-2))
      (bind _returnvalue_4 (AddComponent|NODE_AddOperatorField-5))
      (bind _returnvalue_5 (AddComponent|NODE_AddBoxFalloff-3))
      (bind _returnvalue_6 (AddComponent|NODE_AddCullingField-0))
      (return (Field|SetCullingField _returnvalue_6 (Field|SetBoxFalloff _returnvalue_5 1.0 0.0 1.0 0.0 (Transformation|GetWorldTransform (Variables|Default|GetBoxVolumeCol)) "Field_FallOff_None") (Field|SetOperatorField _returnvalue_4 1.0 (Field|SetBoxFalloff _returnvalue_3 (Variables|Default|GetMagnitude_Input) _falloffminmax_input_x_2 _falloffminmax_input_y_2 0.0 (Transformation|GetWorldTransform (Variables|Default|GetBoxVolumeCol)) (Variables|Default|GetFieldFalloffType_Input)) (Variables|Default|GetOperatorFIeld_Input) "Field_Multiply") "Field_Culling_Outside")))
    (:NewEnumerator1
      (bind _returnvalue_7 (AddComponent|NODE_AddRadialFalloff-1))
      (bind _returnvalue_8 (AddComponent|NODE_AddOperatorField-6))
      (bind _returnvalue_9 (AddComponent|NODE_AddRadialFalloff-2))
      (bind _returnvalue_10 (AddComponent|NODE_AddCullingField-1))
      (return (Field|SetCullingField _returnvalue_10 (Field|SetRadialFalloff _returnvalue_9 1.0 0.0 1.0 0.0 (Components|Sphere|GetScaledSphereRadius (Variables|Default|GetSphereVolumeCol)) (Transformation|GetWorldLocation (Variables|Default|GetSphereVolumeCol)) "Field_FallOff_None") (Field|SetOperatorField _returnvalue_8 1.0 (Field|SetRadialFalloff _returnvalue_7 (Variables|Default|GetMagnitude_Input) _falloffminmax_input_x_3 _falloffminmax_input_y_3 0.0 (Components|Sphere|GetScaledSphereRadius (Variables|Default|GetSphereVolumeCol)) (Transformation|GetWorldLocation (Variables|Default|GetSphereVolumeCol)) (Variables|Default|GetFieldFalloffType_Input)) (Variables|Default|GetOperatorFIeld_Input) "Field_Multiply") "Field_Culling_Outside")))
    (:NewEnumerator2
      (bind _returnvalue_11 (AddComponent|NODE_AddPlaneFalloff-2))
      (bind _returnvalue_12 (AddComponent|NODE_AddOperatorField-7))
      (bind _returnvalue_13 (AddComponent|NODE_AddPlaneFalloff-3))
      (bind _returnvalue_14 (AddComponent|NODE_AddCullingField-2))
      (if (Variables|FieldSettings|AdvancedFieldSettings|PlanarFalloffControls|GetBoxCullingOnPlanar)
        ;[0.2 0.2 0.1412 0.2 32] Alt Planar Falloff Setup
        (bind _returnvalue_15 (AddComponent|NODE_AddBoxFalloff-4))
        (bind _returnvalue_16 (AddComponent|NODE_AddCullingField-3))
        (return (Field|SetCullingField _returnvalue_16 (Field|SetBoxFalloff _returnvalue_15 1.0 0.0 1.0 0.0 (Transformation|GetWorldTransform (Variables|Default|GetBoxVolumeCol)) "Field_FallOff_None") (Field|SetCullingField _returnvalue_14 (Field|SetPlaneFalloff _returnvalue_13 1.0 1.0 1.0 0.0 1e+06 _returnvalue_2 _returnvalue "Field_FallOff_None") (Field|SetOperatorField _returnvalue_12 1.0 (Field|SetPlaneFalloff _returnvalue_11 _magnitude_input _falloffminmax_input_x _falloffminmax_input_y 0.0 _planarfalloffdist _returnvalue_2 _returnvalue _fieldfallofftype_input) _operatorfield_input "Field_Multiply") "Field_Culling_Outside") "Field_Culling_Outside"))
        (else
          (return (Field|SetCullingField _returnvalue_14 (Field|SetPlaneFalloff _returnvalue_13 1.0 1.0 1.0 0.0 1e+06 _returnvalue_2 _returnvalue "Field_FallOff_None") (Field|SetOperatorField _returnvalue_12 1.0 (Field|SetPlaneFalloff _returnvalue_11 _magnitude_input _falloffminmax_input_x _falloffminmax_input_y 0.0 _planarfalloffdist _returnvalue_2 _returnvalue _fieldfallofftype_input) _operatorfield_input "Field_Multiply") "Field_Culling_Outside")))))))

(fn FalloffShapeSwitch
  (params (falloffType :type EFieldFalloffType%20Enum :default "Field_FallOff_None") (falloffMinMax_X :type Float%20%28double-precision%29) (falloffMinMax_Y :type Float%20%28double-precision%29))
  (bind (_falloffminmax_input_x _falloffminmax_input_y) (Math|Vector2D|BreakVector2D (Variables|GetFalloffMinMax_Input_X)))
  (bind (_falloffminmax_input_x_2 _falloffminmax_input_y_2) (Math|Vector2D|BreakVector2D (Variables|GetFalloffMinMax_Input_X)))
  (bind (_falloffminmax_input_x_3 _falloffminmax_input_y_3) (Math|Vector2D|BreakVector2D (Variables|GetFalloffMinMax_Input_X)))
  (bind (_falloffminmax_input_x_4 _falloffminmax_input_y_4) (Math|Vector2D|BreakVector2D (Variables|GetFalloffMinMax_Input_X)))
  (Variables|Default|SetFieldFalloffType_Input falloffType)
  (Variables|SetFalloffMinMax_Input_X (Math|Vector2D|MakeVector2D falloffMinMax_X falloffMinMax_Y))
  (switch Utilities|FlowControl|Switch|SwitchonEFieldShapeType (Variables|Default|GetField%20Falloff%20Shape)
    (:NewEnumerator0
      (bind _returnvalue (AddComponent|NODE_AddBoxFalloff-0))
      (bind _returnvalue_2 (AddComponent|NODE_AddOperatorField-0))
      (return (Field|SetOperatorField _returnvalue_2 1.0 (Field|SetBoxFalloff _returnvalue 1.0 _falloffminmax_input_x_4 _falloffminmax_input_y_4 0.0 (Transformation|GetWorldTransform (Variables|Default|GetBoxVolumeCol)) (Variables|Default|GetFieldFalloffType_Input)) 0 "Field_Multiply")))
    (:NewEnumerator1
      (bind _returnvalue_3 (AddComponent|NODE_AddRadialFalloff-0))
      (bind _returnvalue_4 (AddComponent|NODE_AddOperatorField-1))
      (return (Field|SetOperatorField _returnvalue_4 1.0 (Field|SetRadialFalloff _returnvalue_3 1.0 _falloffminmax_input_x _falloffminmax_input_y 0.0 (Components|Sphere|GetScaledSphereRadius (Variables|Default|GetSphereVolumeCol)) (Transformation|GetWorldLocation (Variables|Default|GetSphereVolumeCol)) (Variables|Default|GetFieldFalloffType_Input)) 0 "Field_Multiply")))
    (:NewEnumerator2
      (if (Variables|FieldSettings|AdvancedFieldSettings|PlanarFalloffControls|GetBoxCullingOnPlanar)
        (bind _returnvalue_5 (AddComponent|NODE_AddPlaneFalloff-1))
        (bind _returnvalue_6 (AddComponent|NODE_AddBoxFalloff-1))
        (bind _returnvalue_7 (AddComponent|NODE_AddOperatorField-3))
        (return (Field|SetOperatorField _returnvalue_7 1.0 (Field|SetBoxFalloff _returnvalue_6 1.0 1.0 1.0 0.0 (Transformation|GetWorldTransform (Variables|Default|GetBoxVolumeCol)) "Field_FallOff_None") (Field|SetPlaneFalloff _returnvalue_5 1.0 _falloffminmax_input_x_2 _falloffminmax_input_y_2 0.0 (Variables|Default|GetPlanarFalloffDist) (Transformation|GetWorldLocation (Variables|Default|GetPlaneVolume)) (Transformation|GetUpVector (Variables|Default|GetPlaneVolume)) (Variables|Default|GetFieldFalloffType_Input)) "Field_Multiply"))
        (else
          (bind _returnvalue_8 (AddComponent|NODE_AddPlaneFalloff-0))
          (bind _returnvalue_9 (AddComponent|NODE_AddOperatorField-2))
          (return (Field|SetOperatorField _returnvalue_9 1.0 (Field|SetPlaneFalloff _returnvalue_8 1.0 _falloffminmax_input_x_3 _falloffminmax_input_y_3 0.0 (Variables|Default|GetPlanarFalloffDist) (Transformation|GetWorldLocation (Variables|Default|GetPlaneVolume)) (Transformation|GetUpVector (Variables|Default|GetPlaneVolume)) (Variables|Default|GetFieldFalloffType_Input)) 0 "Field_Multiply")))))))

(fn ForceMultiplier
  ;[0.2 0.2 0.1412 0.2] This function multiplies our force magnitudes when linear velocity is off. This allows for similar value ranges and more predictable results. The "Force Mult" variable can be adjusted to get desired results.
  (Variables|FieldSettings|ApplyLinearandAngularVelocities|SetRadial%20Magnitude (* (Math|Float|Square (Variables|FieldSettings|ApplyLinearandAngularVelocities|GetRadial%20Magnitude)) (Variables|FieldSettings|AdvancedFieldSettings|VelocityControls|GetForceMult)))
  (Variables|FieldSettings|ApplyLinearandAngularVelocities|SetDirectionalMagnitude (* (Variables|FieldSettings|AdvancedFieldSettings|VelocityControls|GetForceMult) (Math|Float|Square (Variables|FieldSettings|ApplyLinearandAngularVelocities|GetDirectionalMagnitude))))
  (Variables|FieldSettings|ApplyLinearandAngularVelocities|SetTorqueMult (* (Variables|FieldSettings|AdvancedFieldSettings|VelocityControls|GetForceMult) (Math|Float|Square (Variables|FieldSettings|ApplyLinearandAngularVelocities|GetTorqueMult)))))

(fn InitializeFieldVariables
  (bind (_relativescale3d_x _relativescale3d_y _relativescale3d_z) (Math|Vector|BreakVector (Variables|/Script/Engine.SceneComponent|GetRelativeScale3D_X (Variables|Field|GetFieldSystemComponent))))
  ;[0 0.6017 1 0.25] Motion Vector Switch
  (switch Utilities|FlowControl|Switch|SwitchonEFieldForceVel (Variables|FieldSettings|AdvancedFieldSettings|VelocityControls|GetForce/Velocity%20Vector%20Switch)
    (:NewEnumerator0
      (Variables|Default|SetLinearPhysicsType "Field_LinearForce")
      (Variables|Default|SetAngularPhysicsType "Field_AngularTorque")
      ;[0 0.6017 1 0.25] Multiply Force Vector to get semblance of parity with Linear Velocity
      (CallFunction|ForceMultiplier))
    (:NewEnumerator1
      (Variables|Default|SetLinearPhysicsType "Field_LinearVelocity")
      (Variables|Default|SetAngularPhysicsType "Field_AngularVelociy")))
  ;[0 0.6017 1 0.25] Cancel out noise if not activated
  (if (not (Variables|FieldSettings|Noise|GetUseNoise))
    (Variables|FieldSettings|Noise|SetNoiseMinMax "(X=1.000000,Y=1.000000)"))
  ;[0 0.6017 1 0.25] Recalculate delay if using frames
  (if (Variables|FieldSettings|AdvancedFieldSettings|TimingandLifespan|GetUseFramesForTiming)
    (Variables|Default|SetDelayAmount (/ (Variables|Default|GetDelayAmount) (Variables|FieldSettings|AdvancedFieldSettings|TimingandLifespan|GetFPS)))
    (Variables|FieldSettings|AdvancedFieldSettings|TimingandLifespan|SetFieldLifespan (/ (Variables|FieldSettings|AdvancedFieldSettings|TimingandLifespan|GetFieldLifespan) (Variables|FieldSettings|AdvancedFieldSettings|TimingandLifespan|GetFPS))))
  ;[0 0.6017 1 0.25] Setup for planar falloff value.
  (if (> (Variables|FieldSettings|AdvancedFieldSettings|PlanarFalloffControls|GetPlanarFalloffDistOveride) 0)
    (Variables|Default|SetPlanarFalloffDist (Variables|FieldSettings|AdvancedFieldSettings|PlanarFalloffControls|GetPlanarFalloffDistOveride))
    (else
      (Variables|Default|SetPlanarFalloffDist (* _relativescale3d_z 5e+01))))
  ;[0 0.6017 1 0.25] Static Meshes tagged as "SM_Rigids" and Skeletal Meshes tagged "SK_Rigids" will have physics enabled if "Force Dynamic Switching" and "Activate Tagged Static and Skeletal Meshes" are enabled.
  (if (Variables|FieldSettings|AdvancedFieldSettings|ForceDynamicSwitching|GetActivateTaggedStaticAndSkeletal)
    (bind _outactors (Actor|GetAllActorsOfClasswithTag "/Script/Engine.StaticMeshActor" "SM_Rigids"))
    (Variables|Default|SetSimmableStaticMeshes _outactors)
    (bind _outactors_2 (Actor|GetAllActorsOfClasswithTag "/Script/Engine.SkeletalMeshActor" "SK_Rigids"))
    (Variables|Default|SetSimmableSkelMeshes _outactors_2)))

(fn MakeDynamic_EnableNonGC
  (bind (_returnvalue_x _returnvalue_y _returnvalue_z) (Math|Vector|BreakVector (Transformation|GetWorldScale (Variables|Default|GetBoxVolume))))
  (bind (_returnvalue_x_2 _returnvalue_y_2) (Math|Vector|BreakVector (Transformation|GetWorldScale (Variables|Default|GetBoxVolume))))
  (bind (_returnvalue_x_3 _returnvalue_y_3) (Math|Vector|BreakVector (Transformation|GetWorldScale (Variables|Default|GetBoxVolume))))
  (bind (_returnvalue_x_4 _returnvalue_y_4) (Math|Vector|BreakVector (Transformation|GetWorldScale (Variables|Default|GetBoxVolume))))
  (bind (_returnvalue_x_5 _returnvalue_y_5 _returnvalue_z_2) (Math|Vector|BreakVector (Transformation|GetWorldScale (Variables|Default|GetBoxVolume))))
  (bind (_returnvalue_x_6 _returnvalue_y_6) (Math|Vector|BreakVector (Transformation|GetWorldScale (Variables|Default|GetBoxVolume))))
  (bind _planevolume (Variables|Default|GetPlaneVolume))
  (bind (_upvector_x _upvector_y _upvector_z) (Math|Vector|BreakVector (Variables|GetupVector_X)))
  (bind (_upvector_x_2 _upvector_y_2 _upvector_z_2) (Math|Vector|BreakVector (Variables|GetupVector_X)))
  ;[0.2 0.2 0.1412 0.2 32] Stop Checking if all tagged SM/SK meshes have been activated
  (switch Utilities|FlowControl|Switch|SwitchonEFieldActivationType (Variables|Default|GetActivationType)
    (:NewEnumerator0
      _)
    (:NewEnumerator2
      (for _array_element (Variables|Default|GetSimmableStaticMeshes) :!label _label
        (if (not (Physics|IsSimulatingPhysics (Variables|/Script/Engine.StaticMeshActor|GetStaticMeshComponent _array_element) "None"))
          (Variables|FieldSettings|AdvancedFieldSettings|ForceDynamicSwitching|SetActivateTaggedStaticAndSkeletal true)))
      (for _array_element_2 (Variables|Default|GetSimmableSkelMeshes)
        (if (not (Physics|IsSimulatingPhysics (Variables|/Script/Engine.SkeletalMeshActor|GetSkeletalMeshComponent _array_element_2) "None"))
          (Variables|FieldSettings|AdvancedFieldSettings|ForceDynamicSwitching|SetActivateTaggedStaticAndSkeletal true))))
    (:NewEnumerator3
      (goto _label))
    (:NewEnumerator4
      _))
  (if (Variables|FieldSettings|AdvancedFieldSettings|ForceDynamicSwitching|GetActivateTaggedStaticAndSkeletal)
    ;[0.2 0.2 0.1412 0.2 32] Setup Commonly Used Variables
    (Variables|Default|SetupVector (Transformation|GetUpVector _planevolume))
    (Variables|Default|SetforwardVector (Transformation|GetForwardVector _planevolume))
    (Variables|Default|SetrightVector (Transformation|GetRightVector _planevolume))
    (Variables|Default|SetworldLocation (Transformation|GetWorldLocation (Variables|Field|GetFieldSystemComponent)))
    (switch Utilities|FlowControl|Switch|SwitchonEFieldShapeType (Variables|Default|GetField%20Falloff%20Shape)
      (:NewEnumerator0
        ;[0.7344 0.7029 0.0306 0.25 32] Box Volume Check
        (for _array_element_3 (Variables|Default|GetSimmableStaticMeshes)
          (if (and (< (Math|Vector|VectorLength (* (Variables|Default|GetupVector) (Math|Vector|DotProduct (- (Transformation|GetWorldLocation (Variables|Transformation|GetRootComponent _array_element_3)) (Variables|Default|GetworldLocation)) (Variables|Default|GetupVector)))) (* _returnvalue_z_2 5e+01)) (and (< (Math|Vector|VectorLength (* (Variables|Default|GetforwardVector) (Math|Vector|DotProduct (- (Transformation|GetWorldLocation (Variables|Transformation|GetRootComponent _array_element_3)) (Variables|Default|GetworldLocation)) (Variables|Default|GetforwardVector)))) (* _returnvalue_x_4 5e+01)) (< (Math|Vector|VectorLength (* (Variables|Default|GetrightVector) (Math|Vector|DotProduct (- (Transformation|GetWorldLocation (Variables|Transformation|GetRootComponent _array_element_3)) (Variables|Default|GetworldLocation)) (Variables|Default|GetrightVector)))) (* _returnvalue_y_4 5e+01))))
            (Physics|SetSimulatePhysics (Variables|/Script/Engine.StaticMeshActor|GetStaticMeshComponent _array_element_3) true)))
        (for _array_element_4 (Variables|Default|GetSimmableSkelMeshes)
          (if (and (< (Math|Vector|VectorLength (* (Variables|Default|GetupVector) (Math|Vector|DotProduct (- (Transformation|GetWorldLocation (Variables|Transformation|GetRootComponent _array_element_4)) (Variables|Default|GetworldLocation)) (Variables|Default|GetupVector)))) (* _returnvalue_z 5e+01)) (and (< (Math|Vector|VectorLength (* (Variables|Default|GetforwardVector) (Math|Vector|DotProduct (- (Transformation|GetWorldLocation (Variables|Transformation|GetRootComponent _array_element_4)) (Variables|Default|GetworldLocation)) (Variables|Default|GetforwardVector)))) (* _returnvalue_x_3 5e+01)) (< (Math|Vector|VectorLength (* (Variables|Default|GetrightVector) (Math|Vector|DotProduct (- (Transformation|GetWorldLocation (Variables|Transformation|GetRootComponent _array_element_4)) (Variables|Default|GetworldLocation)) (Variables|Default|GetrightVector)))) (* _returnvalue_y_3 5e+01))))
            (Physics|SetSimulatePhysics (Variables|/Script/Engine.SkeletalMeshActor|GetSkeletalMeshComponent _array_element_4) true)))
        (Variables|FieldSettings|AdvancedFieldSettings|ForceDynamicSwitching|SetActivateTaggedStaticAndSkeletal false))
      (:NewEnumerator1
        ;[0.7344 0.7029 0.0306 0.25 32] Sphere Volume Check
        (for _array_element_5 (Variables|Default|GetSimmableStaticMeshes)
          (if (< (Math|Vector|Distance(Vector) (Variables|Default|GetworldLocation) (Transformation|GetWorldLocation (Variables|/Script/Engine.StaticMeshActor|GetStaticMeshComponent _array_element_5))) (Components|Sphere|GetScaledSphereRadius (Variables|Default|GetSphereVolumeCol)))
            (Physics|SetSimulatePhysics (Variables|/Script/Engine.StaticMeshActor|GetStaticMeshComponent _array_element_5) true)))
        (for _array_element_6 (Variables|Default|GetSimmableSkelMeshes)
          (if (< (Math|Vector|Distance(Vector) (Variables|Default|GetworldLocation) (Transformation|GetWorldLocation (Variables|Transformation|GetRootComponent _array_element_6))) (Components|Sphere|GetScaledSphereRadius (Variables|Default|GetSphereVolumeCol)))
            (Physics|SetSimulatePhysics (Variables|/Script/Engine.SkeletalMeshActor|GetSkeletalMeshComponent _array_element_6) true)))
        (Variables|FieldSettings|AdvancedFieldSettings|ForceDynamicSwitching|SetActivateTaggedStaticAndSkeletal false))
      (:NewEnumerator2
        ;[0.7344 0.7029 0.0306 0.25 32] Planar Volume Check
        (for _array_element_7 (Variables|Default|GetSimmableStaticMeshes)
          (bind (_returnvalue_x_7 _returnvalue_y_7 _returnvalue_z_3) (Math|Vector|BreakVector (+ (* (Variables|Default|GetupVector) (Math|Vector|DotProduct (- (Transformation|GetWorldLocation (Variables|Transformation|GetRootComponent _array_element_7)) (Variables|Default|GetworldLocation)) (Variables|Default|GetupVector))) 0)))
          (if (and (and (< (Math|Vector|VectorLength (* (Variables|Default|GetupVector) (Math|Vector|DotProduct (- (Transformation|GetWorldLocation (Variables|Transformation|GetRootComponent _array_element_7)) (Variables|Default|GetworldLocation)) (Variables|Default|GetupVector)))) (Variables|Default|GetPlanarFalloffDist)) (!= (Math|Float|Sign(Float) _returnvalue_z_3) (Math|Float|Sign(Float) _upvector_z_2))) (and (< (Math|Vector|VectorLength (* (Variables|Default|GetforwardVector) (Math|Vector|DotProduct (- (Transformation|GetWorldLocation (Variables|Transformation|GetRootComponent _array_element_7)) (Variables|Default|GetworldLocation)) (Variables|Default|GetforwardVector)))) (* _returnvalue_x_2 5e+01)) (< (Math|Vector|VectorLength (* (Variables|Default|GetrightVector) (Math|Vector|DotProduct (- (Transformation|GetWorldLocation (Variables|Transformation|GetRootComponent _array_element_7)) (Variables|Default|GetworldLocation)) (Variables|Default|GetrightVector)))) (* _returnvalue_y_2 5e+01))))
            (Physics|SetSimulatePhysics (Variables|/Script/Engine.StaticMeshActor|GetStaticMeshComponent _array_element_7) true)))
        (for _array_element_8 (Variables|Default|GetSimmableSkelMeshes)
          (bind (_returnvalue_x_8 _returnvalue_y_8 _returnvalue_z_4) (Math|Vector|BreakVector (+ (* (Variables|Default|GetupVector) (Math|Vector|DotProduct (- (Transformation|GetWorldLocation (Variables|Transformation|GetRootComponent _array_element_8)) (Variables|Default|GetworldLocation)) (Variables|Default|GetupVector))) 0)))
          (if (and (and (< (Math|Vector|VectorLength (* (Variables|Default|GetupVector) (Math|Vector|DotProduct (- (Transformation|GetWorldLocation (Variables|Transformation|GetRootComponent _array_element_8)) (Variables|Default|GetworldLocation)) (Variables|Default|GetupVector)))) (Variables|Default|GetPlanarFalloffDist)) (!= (Math|Float|Sign(Float) _returnvalue_z_4) (Math|Float|Sign(Float) _upvector_z))) (and (< (Math|Vector|VectorLength (* (Variables|Default|GetforwardVector) (Math|Vector|DotProduct (- (Transformation|GetWorldLocation (Variables|Transformation|GetRootComponent _array_element_8)) (Variables|Default|GetworldLocation)) (Variables|Default|GetforwardVector)))) (* _returnvalue_x_6 5e+01)) (< (Math|Vector|VectorLength (* (Variables|Default|GetrightVector) (Math|Vector|DotProduct (- (Transformation|GetWorldLocation (Variables|Transformation|GetRootComponent _array_element_8)) (Variables|Default|GetworldLocation)) (Variables|Default|GetrightVector)))) (* _returnvalue_y_6 5e+01))))
            (Physics|SetSimulatePhysics (Variables|/Script/Engine.SkeletalMeshActor|GetSkeletalMeshComponent _array_element_8) true)))
        (Variables|FieldSettings|AdvancedFieldSettings|ForceDynamicSwitching|SetActivateTaggedStaticAndSkeletal false)))))

(fn SetVisibility
  (bind _debug (Variables|Default|GetDebug))
  (bind _returnvalue (not _debug))
  (bind _field_text (Variables|Default|GetField%20Text))
  (switch Utilities|FlowControl|Switch|SwitchonEFieldShapeType (Variables|Default|GetField%20Falloff%20Shape)
    (:NewEnumerator0
      ;[0.3666 0.07801 0.6823 0.25 32] Swap Falloff field visibility based on user selection
      (Rendering|SetVisibility (Variables|Default|GetSphereVolumeCol) false true)
      (Rendering|SetVisibility (Variables|Default|GetPlaneVolume) false true)
      (Rendering|SetVisibility (Variables|Default|GetBoxVolumeCol) true true))
    (:NewEnumerator1
      (Rendering|SetVisibility (Variables|Default|GetSphereVolumeCol) true true)
      (Rendering|SetVisibility (Variables|Default|GetPlaneVolume) false true)
      (Rendering|SetVisibility (Variables|Default|GetBoxVolumeCol) false true))
    (:NewEnumerator2
      (Rendering|SetVisibility (Variables|Default|GetSphereVolumeCol) false true)
      (Rendering|SetVisibility (Variables|Default|GetPlaneVolume) true true)
      (Rendering|SetVisibility (Variables|Default|GetBoxVolumeCol) false true)
      (if (Variables|FieldSettings|AdvancedFieldSettings|PlanarFalloffControls|GetBoxCullingOnPlanar)
        (Rendering|SetVisibility (Variables|Default|GetPlaneVolumeBox) true false))))
  ;[0.3666 0.07801 0.6823 0.25 32] Debug Mode Visibility Settings for text and falloff volumes
  (Rendering|SetVisibility _field_text (and (Variables|FieldSettings|Display|GetShowDebugText) _debug) true)
  (Development|SetHiddeninGame _field_text _returnvalue true)
  (Development|SetHiddeninGame (Utilities|Array|MakeArray (Variables|Default|GetPlaneVolume) (Variables|Default|GetBoxVolumeCol) (Variables|Default|GetSphereVolumeCol) (Variables|Default|GetSM_DirectionArrow) (Variables|Default|GetSM_RadialArrow)) _returnvalue true)
  ;[0.3666 0.07801 0.6823 0.25 32] Show/Hide Force Arrows
  (Rendering|SetVisibility (Variables|Default|GetSM_DirectionArrow) (Variables|FieldSettings|ApplyLinearandAngularVelocities|GetUseDirectionalVector) true)
  (Rendering|SetVisibility (Variables|Default|GetSM_RadialArrow) (Variables|FieldSettings|ApplyLinearandAngularVelocities|GetUseRadialVector) true)
  (Transformation|SetWorldLocation (Variables|Default|GetSM_RadialArrow) (+ (Transformation|GetWorldLocation (Variables|Field|GetFieldSystemComponent)) (Variables|FieldSettings|AdvancedFieldSettings|VelocityControls|GetRadialPositionOffset)) false false)
  ;[0.3666 0.07801 0.6823 0.25 32] Optional logic to show/hide wireframe and solid visualization (Both are on by default)
  (if (not (Variables|FieldSettings|Display|GetShowSolidShapes))
    (Rendering|SetVisibility (Variables|Default|GetSphereVolume) false false)
    (Rendering|SetVisibility (Variables|Default|GetBoxVolume) false false)
    (Rendering|SetVisibility (Variables|Default|GetPlaneVolume) false false))
  (if (not (Variables|FieldSettings|Display|GetShowWireFrame))
    (Rendering|SetVisibility (Variables|Default|GetSphereVolumeCol) false false)
    (Rendering|SetVisibility (Variables|Default|GetBoxVolumeCol) false false)
    (Rendering|SetVisibility (Variables|Default|GetPlaneVolumeBox) false false)))
