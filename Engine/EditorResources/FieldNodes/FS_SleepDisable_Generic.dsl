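; Generic Field System actor that applies a sleep, disable or kill threshold field to physics within a box, sphere or plane volume.
; While active, each tick adds a transient field of the selected type, using the squared threshold as magnitude.
; FalloffFieldSwitch builds that field by culling the magnitude falloff outside the chosen shape.
; The construction script previews shape, behavior and active state through colored volumes and text, with in-game visibility tied to Debug.
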(event EventBeginPlay
  ;[0.3666 0.07801 0.6823 0.25] Debug Options
  (Development|SetHiddeninGame (Variables|Field|GetFieldSystemComponent) (not (Variables|Default|GetDebug)) true))

(event EventTick
  (params (DeltaSeconds :type Float%20%28single-precision%29))
  ;[1 0.08854 0.08854 0.25] Is Active?
  (if (Variables|Default|GetField%20Active)
    ;[0.9878 0.1849 0.6403 0.2] Sleep Box
    ;;[0.9878 0.1849 0.6403 0.2] Function to use selected falloff volume shape
    (bind _cullingfield (CallFunction|FalloffFieldSwitch (Math|Float|Square (Variables|Default|GetThreshold)) (Variables|Default|GetFaloffType)))
    ;;[0.9878 0.1849 0.6403 0.2] Choose Field Behavior (Default is Sleep)
    (switch Utilities|FlowControl|Switch|SwitchonEFieldSleepType (Variables|Default|GetField%20Behavior)
      (:NewEnumerator0
        (Variables|Default|SetPhysicsType "Field_SleepingThreshold"))
      (:NewEnumerator1
        (Variables|Default|SetPhysicsType "Field_DisableThreshold"))
      (:NewEnumerator2
        (Variables|Default|SetPhysicsType "Field_Kill")))
    ;;[0 1 0.4743 0.251] Field Applied Here
    (Field|AddTransientField (Variables|Field|GetFieldSystemComponent) true (Variables|Default|GetPhysicsType) 0 _cullingfield)))

(fn FalloffFieldSwitch
  (params (Magnitude :type Float%20%28double-precision%29) (FaloffType :type EFieldFalloffType%20Enum :default "Field_FallOff_None"))
  (bind (_falloffminmax_x _falloffminmax_y) (Math|Vector2D|BreakVector2D (Variables|GetFalloffMinMax_X)))
  (bind (_falloffminmax_x_2 _falloffminmax_y_2) (Math|Vector2D|BreakVector2D (Variables|GetFalloffMinMax_X)))
  (bind (_falloffminmax_x_3 _falloffminmax_y_3) (Math|Vector2D|BreakVector2D (Variables|GetFalloffMinMax_X)))
  ;[0.9878 0.1849 0.6403 0.2 32] Output falloff field results according to user selection
  (switch Utilities|FlowControl|Switch|SwitchonEFieldShapeType (Variables|Default|GetField%20Falloff%20Shape)
    (:NewEnumerator0
      (return (Field|SetCullingField (Variables|Default|GetCullingField_Box) (Field|SetBoxFalloff (Variables|Default|GetBoxFalloff) 1.0 0.0 1.0 0.0 (Transformation|GetWorldTransform (Variables|Field|GetFieldSystemComponent)) "Field_FallOff_None") (Field|SetBoxFalloff (Variables|Default|GetBoxFalloff_Magnitude) Magnitude _falloffminmax_x _falloffminmax_y 0.0 (Transformation|GetWorldTransform (Variables|Field|GetFieldSystemComponent)) FaloffType) "Field_Culling_Outside")))
    (:NewEnumerator1
      (return (Field|SetCullingField (Variables|Default|GetCullingField_Sphere) (Field|SetRadialFalloff (Variables|Default|GetRadialFalloff) 1.0 0.0 1.0 0.0 (Components|Sphere|GetScaledSphereRadius (Variables|Default|GetSphereVolumeCol)) (Transformation|GetWorldLocation (Variables|Default|GetSphereVolumeCol)) "Field_FallOff_None") (Field|SetRadialFalloff (Variables|Default|GetRadialFalloff_Magnitude) Magnitude _falloffminmax_x_2 _falloffminmax_y_2 0.0 (Components|Sphere|GetScaledSphereRadius (Variables|Default|GetSphereVolumeCol)) (Transformation|GetWorldLocation (Variables|Default|GetSphereVolumeCol)) FaloffType) "Field_Culling_Outside")))
    (:NewEnumerator2
      (return (Field|SetCullingField (Variables|Default|GetCullingField_Plane) (Field|SetPlaneFalloff (Variables|Default|GetPlaneFalloff) 1.0 1.0 1.0 0.0 1e+06 (Transformation|GetWorldLocation (Variables|Default|GetPlaneVolume)) (Transformation|GetUpVector (Variables|Default|GetPlaneVolume)) "Field_FallOff_None") (Field|SetPlaneFalloff (Variables|Default|GetPlaneFalloff_Magnitude) Magnitude _falloffminmax_x_3 _falloffminmax_y_3 0.0 1e+06 (Transformation|GetWorldLocation (Variables|Default|GetPlaneVolume)) (Transformation|GetUpVector (Variables|Default|GetPlaneVolume)) FaloffType) "Field_Culling_Outside")))))

(fn SetFalloffVisibility
  (switch Utilities|FlowControl|Switch|SwitchonEFieldShapeType (Variables|Default|GetField%20Falloff%20Shape)
    (:NewEnumerator0
      ;[0.3666 0.07801 0.6823 0.25] Swap Falloff field visibility based on user selection
      (Rendering|SetVisibility (Variables|Default|GetSphereVolumeCol) false true)
      (Rendering|SetVisibility (Variables|Default|GetPlaneVolume) false true)
      (Rendering|SetVisibility (Variables|Default|GetBoxVolumeCol) true true))
    (:NewEnumerator1
      (Rendering|SetVisibility (Variables|Default|GetSphereVolumeCol) true true)
      (Rendering|SetVisibility (Variables|Default|GetPlaneVolume) false true)
      (Rendering|SetVisibility (Variables|Default|GetBoxVolumeCol) false true))
    (:NewEnumerator2
      (Rendering|SetVisibility (Variables|Default|GetSphereVolumeCol) false true)
      (Rendering|SetVisibility (Variables|Default|GetPlaneVolume) true true)
      (Rendering|SetVisibility (Variables|Default|GetBoxVolumeCol) false true)))
  ;[0.3666 0.07801 0.6823 0.25] Optional logic to show/hide wireframe and solid visualization
  (if (not (Variables|Display|GetShowSolidShapes))
    (Rendering|SetVisibility (Variables|Default|GetSphereVolume) false false)
    (Rendering|SetVisibility (Variables|Default|GetBoxVolume) false false)
    (Rendering|SetVisibility (Variables|Default|GetPlaneVolume) false false))
  (if (not (Variables|Display|GetShowWireFrame))
    (Rendering|SetVisibility (Variables|Default|GetSphereVolumeCol) false false)
    (Rendering|SetVisibility (Variables|Default|GetBoxVolumeCol) false false)
    (Rendering|SetVisibility (Variables|Default|GetPlaneVolumeBoxCol) false false)))

(fn ConstructionScript
  (bind (_returnvalue_x _returnvalue_y _returnvalue_z) (Math|Vector|BreakVector (Transformation|GetWorldScale (Variables|Field|GetFieldSystemComponent))))
  (bind _debug (Variables|Default|GetDebug))
  (bind _returnvalue (not _debug))
  (bind (_returnvalue_x_2 _returnvalue_y_2 _returnvalue_z_2) (Math|Vector|BreakVector (Transformation|GetWorldScale (Variables|Field|GetFieldSystemComponent))))
  (bind _fieldsystemcomponent (Variables|Field|GetFieldSystemComponent))
  (bind (_returnvalue_x_3 _returnvalue_y_3 _returnvalue_z_3) (Math|Vector|BreakVector (Transformation|GetWorldScale _fieldsystemcomponent)))
  (bind _thresholdvaluetext (Variables|Default|GetThresholdValueText))
  (bind _tresholdtitletext (Variables|Default|GetTresholdTitleText))
  (bind _activevaluetext (Variables|Default|GetActiveValueText))
  (bind _activetitletext (Variables|Default|GetActiveTitleText))
  (bind _sleeptitletext (Variables|Default|GetSleepTitleText))
  (bind _array (Utilities|Array|MakeArray _sleeptitletext _activetitletext _activevaluetext _tresholdtitletext _thresholdvaluetext))
  (bind _sleep_text (Variables|Components|GetSleep%20Text))
  (bind _killcolour (Variables|Display|GetKillColour))
  (bind _disablecolour (Variables|Display|GetDisableColour))
  (bind _sleep_colour (Variables|Display|GetSleep%20Colour))
  ;[0.3666 0.07801 0.6823 0.25] Preview Logic
  (Rendering|SetVisibility _sleep_text (and (Variables|Display|GetShowDebugText) _debug) true)
  (Development|SetHiddeninGame _sleep_text _returnvalue true)
  ;;[0.3666 0.07801 0.6823 0.25] Function to swap fallloff shape visibility according to selected falloff type
  (CallFunction|SetFalloffVisibility)
  ;;[0.3666 0.07801 0.6823 0.25] Text Visibility Management
  (Development|SetHiddeninGame (Utilities|Array|MakeArray (Variables|Default|GetPlaneVolume) (Variables|Default|GetBoxVolumeCol) (Variables|Default|GetSphereVolumeCol) (Variables|Default|GetTresholdTitleText) (Variables|Default|GetThresholdValueText)) _returnvalue true)
  ;;[0.3666 0.07801 0.6823 0.25] Volume Color and Shape Swapping
  (switch Utilities|FlowControl|Switch|SwitchonEFieldSleepType (Variables|Default|GetField%20Behavior)
    (:NewEnumerator0
      (Variables|Display|SetSleep%20Colour%20Def _sleep_colour)
      (Components|Arrow|SetArrowColor(LinearColor) (Variables|Default|GetArrow) _sleep_colour)
      (Rendering|Components|TextRender|SetText (Variables|Default|GetSleepTitleText) (Variables|Default|GetSleepText)))
    (:NewEnumerator1
      (Variables|Display|SetSleep%20Colour%20Def _disablecolour)
      (Components|Arrow|SetArrowColor(LinearColor) (Variables|Default|GetArrow) _disablecolour)
      (Rendering|Components|TextRender|SetText (Variables|Default|GetSleepTitleText) (Variables|Default|GetDisableText)))
    (:NewEnumerator2
      (Variables|Display|SetSleep%20Colour%20Def _killcolour)
      (Components|Arrow|SetArrowColor(LinearColor) (Variables|Default|GetArrow) _killcolour)
      (Rendering|Components|TextRender|SetText (Variables|Default|GetSleepTitleText) (Variables|Default|GetKillText))))
  (Variables|Default|SetDeactivated (* (Variables|Display|GetSleep%20Colour%20Def) "(R=0.250000,G=0.250000,B=0.250000,A=1.000000)"))
  (switch Utilities|FlowControl|Switch|SwitchonEFieldShapeType (Variables|Default|GetField%20Falloff%20Shape)
    (:NewEnumerator0
      (Variables|Default|SetFieldVolume (Variables|Default|GetBoxVolume)))
    (:NewEnumerator1
      (Transformation|SetWorldScale3D (Variables|Default|GetSphereVolumeCol) (Math|Conversions|ToVector(Float) _returnvalue_z))
      (Variables|Default|SetFieldVolume (Variables|Default|GetSphereVolume)))
    (:NewEnumerator2
      (Variables|Default|SetFieldVolume (Variables|Default|GetPlaneVolume))))
  (bind _returnvalue_2 (Rendering|Material|CreateDynamicMaterialInstance (Variables|Default|GetFieldVolume) 0 "/Engine/EditorResources/FieldNodes/_Resources/MI_FieldRadiusPreview.MI_FieldRadiusPreview" "None" :!class /Script/Engine.PrimitiveComponent))
  ;;[0.3666 0.07801 0.6823 0.25] Swap colour based on active status
  (if (Variables|Default|GetField%20Active)
    (Rendering|Material|SetVectorParameterValue _returnvalue_2 "Preview_Colour" (Variables|Display|GetSleep%20Colour%20Def) :!class /Script/Engine.MaterialInstanceDynamic)
    (Rendering|Components|TextRender|SetTextRenderColor _array (Math|Conversions|ToColor(LinearColor) (Variables|Display|GetSleep%20Colour%20Def)))
    (else
      (Rendering|Material|SetVectorParameterValue _returnvalue_2 "Preview_Colour" (Variables|Default|GetDeactivated) :!class /Script/Engine.MaterialInstanceDynamic)
      (Rendering|Components|TextRender|SetTextRenderColor _array (Math|Conversions|ToColor(LinearColor) (Variables|Default|GetDeactivatedText)))))
  ;;[0.3666 0.07801 0.6823 0.25] Set Debug Text Value
  (Rendering|Components|TextRender|SetText (Variables|Default|GetActiveValueText) (Utilities|Text|ToText(Boolean) (Variables|Default|GetField%20Active)))
  (Rendering|Components|TextRender|SetText (Variables|Default|GetThresholdValueText) (Utilities|Text|ToText(Float) (Variables|Default|GetThreshold)))
  ;;[0.3666 0.07801 0.6823 0.25] Scale and Position Debug Text
  (switch Utilities|FlowControl|Switch|SwitchonEFieldShapeType (Variables|Default|GetField%20Falloff%20Shape)
    (:NewEnumerator0
      (Transformation|SetWorldScale3D (Variables|Components|GetSleep%20Text) (Math|Conversions|ToVector(Float) (* (* (Math|Float|Min(Float) _returnvalue_x_2 _returnvalue_y_2 _returnvalue_z_2) 0.5) (Variables|Display|GetTextScaleMult)))))
    (:NewEnumerator1
      (Transformation|SetWorldScale3D (Variables|Components|GetSleep%20Text) (Math|Conversions|ToVector(Float) _returnvalue_z_3)))
    (:NewEnumerator2
      (Transformation|SetWorldScale3D (Variables|Components|GetSleep%20Text) (Math|Conversions|ToVector(Float) (* (Math|Float|Min(Float) _returnvalue_x_3 _returnvalue_y_3) 0.5)))))
  (Transformation|SetRelativeLocation (Variables|Components|GetSleep%20Text) (Math|Vector|MakeVector 0.0 0.0 (Variables|Display|GetText%20Vertical%20Offset)) false false))
