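; FS_BaseField is a field system Blueprint that provides shared falloff, noise and dynamic state functions for building fields.
; Box and sphere components define box, plane or radial falloff, multiplied into an input field with optional culling outside the shape.
; Noise is either added as a vector, built from three differently rotated per-axis noise fields, or multiplied as a scalar.
; Tick can sync the noise transform to the field component.
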(event EventTick
  (params (DeltaSeconds :type Float%20%28single-precision%29))
  (Utilities|FlowControl|Sequence
    (:then_0
      (if (Variables|Noise|GetNoise%20Use%20Actor%20Location)
        (Variables|Noise|SetNoiseLocationInternal (Transformation|GetWorldLocation (Variables|Field|GetFieldSystemComponent)))))
    (:then_1
      (if (Variables|Noise|GetNoise%20Use%20Actor%20Rotation)
        (Variables|Noise|SetNoiseRotationInternal (Transformation|GetWorldRotation (Variables|Field|GetFieldSystemComponent)))))
    (:then_2
      (if (Variables|Noise|GetNoise%20Use%20Actor%20Scale)
        (Variables|Noise|SetNoiseScaleInternal (Transformation|GetWorldScale (Variables|Field|GetFieldSystemComponent)))))))

(fn ApplyFalloff
  (params (FieldIn :type Field%20Node%20Base%20Object%20Reference))
  (bind _falloffapplyoperatorfield (Variables|Default|GetFalloffApplyOperatorField))
  (bind _fallofffield (CallFunction|GetFalloffField))
  (if (Variables|Falloff|GetCull%20Outside%20Falloff)
    (bind _cullingfallofffield (CallFunction|GetCullingFalloffField))
    (return (Field|SetCullingField (Variables|Default|GetFalloffCullingField) _cullingfallofffield (Field|SetOperatorField _falloffapplyoperatorfield 1.0 FieldIn _fallofffield "Field_Multiply") "Field_Culling_Outside"))
    (else
      (return (Field|SetOperatorField _falloffapplyoperatorfield 1.0 FieldIn _fallofffield "Field_Multiply")))))

(fn ApplyNoise
  (params (FieldIn :type Field%20Node%20Base%20Object%20Reference))
  (if (Utilities|Enum|Equal(Enum) (Variables|Noise|GetNoise%20Mode) "NewEnumerator0")
    (if (> (Variables|Noise|GetAdd%20Noise%20Gain) 0)
      (bind _vectornoisefield (CallFunction|VectorNoise))
      (return (Field|SetOperatorField (Variables|Default|GetNoiseApplyOperatorField) 1.0 FieldIn _vectornoisefield "Field_Add"))
      (else
        (return FieldIn)))
    (else
      (if (and (>= (Variables|Noise|GetMult%20Noise%20Min) 0) (> (Variables|Noise|GetMult%20Noise%20Max) (Variables|Noise|GetMult%20Noise%20Min)))
        (bind _scalarnoisefield (CallFunction|ScalarNoise))
        (return (Field|SetOperatorField (Variables|Default|GetNoiseApplyOperatorField) 1.0 FieldIn _scalarnoisefield "Field_Multiply"))
        (else
          (return FieldIn))))))

(fn BoxFalloff
  (return (Field|SetBoxFalloff (Variables|Default|GetBoxFalloffField) 1.0 (Variables|Falloff|GetMin%20Falloff) (Variables|Falloff|GetMax%20Falloff) (Variables|Falloff|GetMin%20Falloff) (Math|Transform|MakeTransform (Transformation|GetWorldLocation (Variables|Default|GetBox)) (Transformation|GetWorldRotation (Variables|Default|GetBox)) (/ (Components|Box|GetScaledBoxExtent (Variables|Default|GetBox)) "50.000000,50.000000,50.000000")) (Variables|Falloff|GetFalloff%20Type)) (Field|SetBoxFalloff (Variables|Default|GetBoxCullingFalloffField) 1.0 0.0 1.0 0.0 (Math|Transform|MakeTransform (Transformation|GetWorldLocation (Variables|Default|GetBox)) (Transformation|GetWorldRotation (Variables|Default|GetBox)) (/ (Components|Box|GetScaledBoxExtent (Variables|Default|GetBox)) "50.000000,50.000000,50.000000")) "Field_FallOff_None")))

(fn GetCullingFalloffField
  (bind (_planefalloff _planecullingfalloff) (CallFunction|PlaneFalloff))
  (bind (_radialfalloff _radialcullingfalloff) (CallFunction|RadialFalloff))
  (bind (_boxfalloff _boxcullingfalloff) (CallFunction|BoxFalloff))
  (return (Utilities|Select :Index (Variables|Falloff|GetFalloff%20Shape) :NewEnumerator0 _boxcullingfalloff :NewEnumerator1 _radialcullingfalloff :NewEnumerator2 _planecullingfalloff)))

(fn GetDynamicStateField
  (if (Variables|Falloff|GetCull%20Outside%20Falloff)
    (bind _cullingfallofffield (CallFunction|GetCullingFalloffField))
    (return (Field|SetCullingField (Variables|Default|GetDynamicStateCullingField) _cullingfallofffield (Field|SetUniformInteger (Variables|Default|GetDynamicStateField) (Math|Conversions|ToInteger(Byte) (Variables|Default|GetDynamicState))) "Field_Culling_Outside"))
    (else
      (return (Field|SetUniformInteger (Variables|Default|GetDynamicStateField) (Math|Conversions|ToInteger(Byte) (Variables|Default|GetDynamicState)))))))

(fn GetFalloffField
  (bind _planefalloff (CallFunction|PlaneFalloff))
  (bind _radialfalloff (CallFunction|RadialFalloff))
  (bind _boxfalloff (CallFunction|BoxFalloff))
  (return (Utilities|Select :Index (Variables|Falloff|GetFalloff%20Shape) :NewEnumerator0 _boxfalloff :NewEnumerator1 _radialfalloff :NewEnumerator2 _planefalloff)))

(fn PlaneFalloff
  (return (Field|SetPlaneFalloff (Variables|Default|GetPlaneFalloffField) 1.0 (Variables|Falloff|GetMin%20Falloff) (Variables|Falloff|GetMax%20Falloff) (Variables|Falloff|GetMin%20Falloff) (.x (* (Components|Box|GetScaledBoxExtent (Variables|Default|GetBox)) "2.000000,2.000000,2.000000")) (- (Transformation|GetWorldLocation (Variables|Default|GetBox)) (Utilities|Operators|Multiply (Math|Vector|Normalize (Transformation|GetForwardVector (Variables|Default|GetBox))) (.x (* (Components|Box|GetScaledBoxExtent (Variables|Default|GetBox)) "2.000000,2.000000,2.000000")) 0.5)) (* (Math|Vector|Normalize (Transformation|GetForwardVector (Variables|Default|GetBox))) "-1.000000,-1.000000,-1.000000") (Variables|Falloff|GetFalloff%20Type)) (Field|SetPlaneFalloff (Variables|Default|GetPlaneCullingFalloffField) 1.0 0.0 1.0 0.0 (.x (* (Components|Box|GetScaledBoxExtent (Variables|Default|GetBox)) "2.000000,2.000000,2.000000")) (- (Transformation|GetWorldLocation (Variables|Default|GetBox)) (Utilities|Operators|Multiply (Math|Vector|Normalize (Transformation|GetForwardVector (Variables|Default|GetBox))) (.x (* (Components|Box|GetScaledBoxExtent (Variables|Default|GetBox)) "2.000000,2.000000,2.000000")) 0.5)) (* (Math|Vector|Normalize (Transformation|GetForwardVector (Variables|Default|GetBox))) "-1.000000,-1.000000,-1.000000") "Field_FallOff_None")))

(fn RadialFalloff
  (return (Field|SetRadialFalloff (Variables|Default|GetRadialFalloffField) 1.0 (Variables|Falloff|GetMin%20Falloff) (Variables|Falloff|GetMax%20Falloff) (Variables|Falloff|GetMin%20Falloff) (Components|Sphere|GetScaledSphereRadius (Variables|Default|GetSphere)) (Transformation|GetWorldLocation (Variables|Default|GetSphere)) (Variables|Falloff|GetFalloff%20Type)) (Field|SetRadialFalloff (Variables|Default|GetRadialCullingFalloffField) 1.0 0.0 1.0 0.0 (Components|Sphere|GetScaledSphereRadius (Variables|Default|GetSphere)) (Transformation|GetWorldLocation (Variables|Default|GetSphere)) "Field_FallOff_None")))

(fn ScalarNoise
  (return (Field|SetNoiseField (Variables|Default|GetNoiseXField) (Variables|Noise|GetMult%20Noise%20Min) (Variables|Noise|GetMult%20Noise%20Max) (Math|Transform|MakeTransform (Variables|Noise|GetNoiseLocationInternal) (Variables|Noise|GetNoiseRotationInternal) (Math|Conversions|ToVector(Float) (Utilities|Operators|Multiply (Math|Vector|GetMaxElement (Variables|Noise|GetNoiseScaleInternal)) (Variables|Noise|GetNoiseScaleInternalBase) (Variables|Noise|GetNoise%20Scale%20Mult)))))))

(fn ConstructionScript
  (bind _debug (Variables|Default|GetDebug))
  (bind _returnvalue (not _debug))
  (bind _falloff_shape (Variables|Falloff|GetFalloff%20Shape))
  (bind _returnvalue_2 (Utilities|Enum|Equal(Enum) _falloff_shape "NewEnumerator1"))
  (Rendering|SetVisibility (Variables|Default|GetSphere) _returnvalue_2 false)
  (Rendering|SetVisibility (Variables|Default|GetBox) (not _returnvalue_2) false)
  (Development|SetHiddeninGame (Variables|Default|GetSphere) _returnvalue false)
  (Development|SetHiddeninGame (Variables|Default|GetBox) _returnvalue false))

(fn VectorNoise
  (bind (_x _y) (Math|Vector2D|BreakVector2D (Variables|Noise|GetNoise%20Gain%20Range)))
  (bind (_x_2 _y_2) (Math|Vector2D|BreakVector2D (Variables|Noise|GetNoise%20Gain%20Range)))
  (bind (_x_3 _y_3) (Math|Vector2D|BreakVector2D (Variables|Noise|GetNoise%20Gain%20Range)))
  (Variables|Noise|SetNoise%20Gain%20Range (Math|Vector2D|MakeVector2D (* (Variables|Noise|GetAdd%20Noise%20Gain) -1.0) (Variables|Noise|GetAdd%20Noise%20Gain)))
  (return (Field|SetOperatorField (Variables|Default|GetVectorNoiseField) 1.0 (Field|SetOperatorField (Variables|Default|GetNoiseXYOperatorField) 1.0 (Field|SetOperatorField (Variables|Default|GetNoiseXOperatorField) 1.0 (Field|SetUniformVector (Variables|Default|GetNoiseXDirectionField) 1.0 "1.000000, 0, 0") (Field|SetNoiseField (Variables|Default|GetNoiseXField) _x_2 _y_2 (Math|Transform|MakeTransform (Variables|Noise|GetNoiseLocationInternal) (Variables|Noise|GetNoiseRotationInternal) (Math|Conversions|ToVector(Float) (Utilities|Operators|Multiply (Math|Vector|GetMaxElement (Variables|Noise|GetNoiseScaleInternal)) (Variables|Noise|GetNoiseScaleInternalBase) (Variables|Noise|GetNoise%20Scale%20Mult))))) "Field_Multiply") (Field|SetOperatorField (Variables|Default|GetNoiseYOperatorField) 1.0 (Field|SetUniformVector (Variables|Default|GetNoiseYDirectionField) 1.0 "0.000000,1.000000, 0") (Field|SetNoiseField (Variables|Default|GetNoiseYField) _x_3 _y_3 (Math|Transform|MakeTransform (Variables|Noise|GetNoiseLocationInternal) (Math|Rotator|CombineRotators (Variables|Noise|GetNoiseRotationInternal) (Math|Rotator|MakeRotator 1e+01 2e+01 3e+01)) (Math|Conversions|ToVector(Float) (Utilities|Operators|Multiply (Math|Vector|GetMaxElement (Variables|Noise|GetNoiseScaleInternal)) (Variables|Noise|GetNoiseScaleInternalBase) (Variables|Noise|GetNoise%20Scale%20Mult))))) "Field_Multiply") "Field_Add") (Field|SetOperatorField (Variables|Default|GetNoiseZOperatorField) 1.0 (Field|SetUniformVector (Variables|Default|GetNoiseZDirectionField) 1.0 "0.000000, 0,1.000000") (Field|SetNoiseField (Variables|Default|GetNoiseZField) _x _y (Math|Transform|MakeTransform (Variables|Noise|GetNoiseLocationInternal) (Math|Rotator|CombineRotators (Variables|Noise|GetNoiseRotationInternal) (Math|Rotator|MakeRotator 1.2e+02 1.3e+02 1.4e+02)) (Math|Conversions|ToVector(Float) (Utilities|Operators|Multiply (Math|Vector|GetMaxElement (Variables|Noise|GetNoiseScaleInternal)) (Variables|Noise|GetNoiseScaleInternalBase) (Variables|Noise|GetNoise%20Scale%20Mult))))) "Field_Multiply") "Field_Add")))
