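; Field System actor, FS_RadialField, that pushes physics objects with a radial vector field centered on its Sphere component.
; BeginPlay maps the chosen Physics Vector enum to a force, velocity or torque field type, defaulting to linear force.
; Each tick, while enabled, it builds the radial vector, applies noise and falloff, then submits it as a transient field, optionally adding a dynamic-state field.
; The construction script shows InnerSphere in game only when Debug is set.
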(event EventBeginPlay
  (if (Utilities|Enum|Equal(Enum) (Variables|Default|GetPhysics%20Vector) "Vector_AngularTorque")
    (Variables|Default|SetPhysics%20Type "Field_AngularTorque")
    (elif (Utilities|Enum|Equal(Enum) (Variables|Default|GetPhysics%20Vector) "Vector_AngularVelocity")
      (Variables|Default|SetPhysics%20Type "Field_AngularVelociy")
      (elif (Utilities|Enum|Equal(Enum) (Variables|Default|GetPhysics%20Vector) "Vector_LinearVelocity")
        (Variables|Default|SetPhysics%20Type "Field_LinearVelocity")
        (else
          (Variables|Default|SetPhysics%20Type "Field_LinearForce"))))))

(event EventTick
  (params (DeltaSeconds :type Float%20%28single-precision%29))
  (|CallParentFunction DeltaSeconds)
  (if (Variables|Default|GetEnable%20Field)
    (Utilities|FlowControl|Sequence
      (:then_0
        (bind _returnvalue (AddComponent|NODE_AddRadialVector-0))
        (Variables|Default|SetVectorField (Field|SetRadialVector _returnvalue (Variables|Default|GetField%20Magnitude) (Transformation|GetWorldLocation (Variables|Default|GetSphere)))))
      (:then_1
        (bind _fieldout (CallFunction|ApplyNoise (Variables|Default|GetVectorField)))
        (bind _fieldout_2 (CallFunction|ApplyFalloff _fieldout))
        (Field|AddTransientField (Variables|Field|GetFieldSystemComponent) true (Variables|Default|GetPhysics%20Type) 0 _fieldout_2)
        (if (Variables|Default|GetForce%20Physics%20Dynamic)
          (bind _dynamicstatefield (CallFunction|GetDynamicStateField))
          (Field|AddTransientField (Variables|Field|GetFieldSystemComponent) true "Field_DynamicState" 0 _dynamicstatefield))))))

(fn ConstructionScript
  (|CallParentFunction)
  (Development|SetHiddeninGame (Variables|Default|GetInnerSphere) (not (Variables|Default|GetDebug)) false))
