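; Field System Blueprint that applies a wave-shaped scalar field to Chaos physics at its sphere's location.
; BeginPlay maps the selected scalar option to a field physics type, defaulting to internal cluster strain.
; While enabled, each tick adds a new WaveScalar node component, passes it through noise and falloff, and submits it as a transient field.
; An optional dynamic-state field follows when forcing physics dynamic.
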(event EventBeginPlay
  (if (Utilities|Enum|Equal(Enum) (Variables|Default|GetPhysics%20Scalar) "Scalar_ExternalClusterStrain")
    (Variables|Default|SetPhysics%20Type "Field_ExternalClusterStrain")
    (elif (Utilities|Enum|Equal(Enum) (Variables|Default|GetPhysics%20Scalar) "Scalar_Kill")
      (Variables|Default|SetPhysics%20Type "Field_Kill")
      (elif (Utilities|Enum|Equal(Enum) (Variables|Default|GetPhysics%20Scalar) "Scalar_DisableThreshold")
        (Variables|Default|SetPhysics%20Type "Field_DisableThreshold")
        (elif (Utilities|Enum|Equal(Enum) (Variables|Default|GetPhysics%20Scalar) "Scalar_SleepingThreshold")
          (Variables|Default|SetPhysics%20Type "Field_SleepingThreshold")
          (else
            (Variables|Default|SetPhysics%20Type "Field_InternalClusterStrain")))))))

(event EventTick
  (params (DeltaSeconds :type Float%20%28single-precision%29))
  (|CallParentFunction DeltaSeconds)
  (if (Variables|Default|GetEnable%20Field)
    (Utilities|FlowControl|Sequence
      (:then_0
        (bind _returnvalue (AddComponent|NODE_AddWaveScalar-1))
        (Variables|Default|SetScalarField (Field|SetWaveScalar _returnvalue (Variables|Default|GetField%20Magnitude) (Transformation|GetWorldLocation (Variables|Default|GetSphere)) (Variables|Default|GetWave%20Length) (Variables|Default|GetWave%20Period) (Variables|Default|GetWave%20Function) (Variables|Default|GetWave%20Falloff%20Type))))
      (:then_1
        (bind _fieldout (CallFunction|ApplyNoise (Variables|Default|GetScalarField)))
        (bind _fieldout_2 (CallFunction|ApplyFalloff _fieldout))
        (Field|AddTransientField (Variables|Field|GetFieldSystemComponent) true (Variables|Default|GetPhysics%20Type) 0 _fieldout_2)
        (if (Variables|Default|GetForce%20Physics%20Dynamic)
          (bind _dynamicstatefield (CallFunction|GetDynamicStateField))
          (Field|AddTransientField (Variables|Field|GetFieldSystemComponent) true "Field_DynamicState" 0 _dynamicstatefield))))))

(fn ConstructionScript
  (|CallParentFunction))
